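# Frontend physics asset for the male medium Maroon Belt ensemble body, defining capsules and joints for its dynamic skirt and legs.
# Six skirt chains (front, side and back on each side) hang from the thighs through tiny capsules, with mostly tight twist and moderate swing limits along each chain.
# Thigh, calf and foot capsules link to the root by free angular joints.
# An oversized root capsule below the origin appears to act as a floor collider.
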PhysicsAsset /BRCosmetics/Characters/Player/Male/Medium/Bodies/M_MED_Ensemble_Maroon_Belt/Meshes/Frontend_M_MED_Ensemble_Maroon_Belt_Physics2
bodies {
 calf_l [{shapeType=Capsule, length=39.5, radius=7.5, shapeName=calf_r_capsule, center=<-26.4668 -2.0022 0>, rotation=<rot 85 -5.0001 0.0002>}]
 calf_r [{shapeType=Capsule, length=42, radius=9, shapeName=calf_r_capsule, center=<28.6541 0 0>, rotation=<rot 90 0 0>}]
 dyn_skirtH_front_l_02 [{shapeType=Capsule, length=1, radius=1, shapeName=dyn_skirtH_front_l_02_capsule, center=<5 0 10>, rotation=<rot 0 0 90>}]
 dyn_skirtH_front_l_03 [{shapeType=Capsule, length=18, radius=6.25, shapeName=dyn_skirtH_front_l_03_capsule, center=<5 0 5>, rotation=<rot 0 0 90>}]
 dyn_skirtH_front_l_04 [{shapeType=Capsule, length=17.75, radius=6.5, shapeName=dyn_skirtH_front_l_04_capsule, center=<5 0 5>, rotation=<rot 0 0 90>}]
 dyn_skirtH_front_l_05 [{shapeType=Capsule, length=15, radius=9.25, shapeName=dyn_skirtH_front_l_05_capsule, center=<5 0 5>, rotation=<rot 0 0 90>}]
 dyn_skirtH_front_l_06 [{shapeType=Capsule, length=14.75, radius=9.5, shapeName=dyn_skirtH_front_l_06_capsule, center=<5 0 5>, rotation=<rot 0 0 90>}]
 dyn_skirt_bk_l_02 [{shapeType=Capsule, length=1, radius=1, shapeName=dyn_skirt_bk_l_02_capsule, center=<5 0 10>, rotation=<rot 0 0 90>}]
 dyn_skirt_bk_l_03 [{shapeType=Capsule, length=12, radius=6.5, shapeName=dyn_skirt_bk_l_03_capsule, center=<5 0 5>, rotation=<rot 0 0 90>}]
 dyn_skirt_bk_l_04 [{shapeType=Capsule, length=9.75, radius=8.75, shapeName=dyn_skirt_bk_l_04_capsule, center=<5 0 5>, rotation=<rot 0 0 90>}]
 dyn_skirt_bk_l_05 [{shapeType=Capsule, length=17.75, radius=9.75, shapeName=dyn_skirt_bk_l_05_capsule, center=<5 0 5>, rotation=<rot 0 0 90>}]
 dyn_skirt_bk_l_06 [{shapeType=Capsule, length=17.25, radius=10.25, shapeName=dyn_skirt_l_06_capsule, center=<5 0 5>, rotation=<rot 0 0 90>}]
 dyn_skirt_bk_r_02 [{shapeType=Capsule, length=1, radius=1, shapeName=dyn_skirt_bk_r_02_capsule, center=<-5 0 -10>, rotation=<rot 0 0 90>}]
 dyn_skirt_bk_r_03 [{shapeType=Capsule, length=15.5, radius=6.25, shapeName=dyn_skirt_bk_r_03_capsule, center=<-5 0 -5>, rotation=<rot 0 0 90>}]
 dyn_skirt_bk_r_04 [{shapeType=Capsule, length=12.75, radius=9, shapeName=dyn_skirt_bk_r_04_capsule, center=<-5 0 -5>, rotation=<rot 0 0 90>}]
 dyn_skirt_bk_r_05 [{shapeType=Capsule, length=19.75, radius=11, shapeName=dyn_skirt_bk_r_05_capsule, center=<-5 0 -5>, rotation=<rot 0 0 90>}]
 dyn_skirt_bk_r_06 [{shapeType=Capsule, length=20.5, radius=10.25, shapeName=dyn_skirt_front_r_06_capsule, center=<-5 0 -5>, rotation=<rot 0 0 90>}]
 dyn_skirt_front_r_02 [{shapeType=Capsule, length=1, radius=1, shapeName=dyn_skirt_front_r_02_capsule, center=<-5 0 -10>, rotation=<rot 0 0 90>}]
 dyn_skirt_front_r_03 [{shapeType=Capsule, length=10.75, radius=7.75, shapeName=dyn_skirt_front_r_03_capsule, center=<-5 0 -5>, rotation=<rot 0 0 90>}]
 dyn_skirt_front_r_04 [{shapeType=Capsule, length=10.75, radius=9.5, shapeName=dyn_skirt_front_r_04_capsule, center=<-5 0 -5>, rotation=<rot 0 0 90>}]
 dyn_skirt_front_r_05 [{shapeType=Capsule, length=10.75, radius=10, shapeName=dyn_skirt_front_r_05_capsule, center=<-5 0 -5>, rotation=<rot 0 0 90>}]
 dyn_skirt_front_r_06 [{shapeType=Capsule, length=10.75, radius=12.25, shapeName=dyn_skirt_front_r_06_capsule, center=<-5 0 -5>, rotation=<rot 0 0 90>}]
 dyn_skirt_l_02 [{shapeType=Capsule, length=1, radius=1, shapeName=dyn_skirt_l_02_capsule, center=<5 0 10>, rotation=<rot 0 0 90>}]
 dyn_skirt_l_03 [{shapeType=Capsule, length=25.25, radius=4.75, shapeName=dyn_skirt_l_03_capsule, center=<5 0 5>, rotation=<rot 0 0 90>}]
 dyn_skirt_l_04 [{shapeType=Capsule, length=24.5, radius=5.5, shapeName=dyn_skirt_l_04_capsule, center=<5 0 5>, rotation=<rot 0 0 90>}]
 dyn_skirt_l_05 [{shapeType=Capsule, length=31, radius=8, shapeName=dyn_skirt_l_05_capsule, center=<5 0 5>, rotation=<rot 0 0 90>}]
 dyn_skirt_l_06 [{shapeType=Capsule, length=30, radius=9, shapeName=dyn_skirt_l_06_capsule, center=<5 0 5>, rotation=<rot 0 0 90>}]
 dyn_skirt_r_02 [{shapeType=Capsule, length=1, radius=1, shapeName=dyn_skirt_r_02_capsule, center=<-5 0 -10>, rotation=<rot 0 0 90>}]
 dyn_skirt_r_03 [{shapeType=Capsule, length=18, radius=4.75, shapeName=dyn_skirt_r_03_capsule, center=<-5 0 -5>, rotation=<rot 0 0 90>}]
 dyn_skirt_r_04 [{shapeType=Capsule, length=16.5, radius=6.25, shapeName=dyn_skirt_r_04_capsule, center=<-5 0 -5>, rotation=<rot 0 0 90>}]
 dyn_skirt_r_05 [{shapeType=Capsule, length=43, radius=7.25, shapeName=dyn_skirt_r_05_capsule, center=<-5 0 -5>, rotation=<rot 0 0 90>}]
 dyn_skirt_r_06 [{shapeType=Capsule, length=42.25, radius=8, shapeName=dyn_skirt_r_06_capsule, center=<-5 0 -5>, rotation=<rot 0 0 90>}]
 foot_l [{
   shapeType=Capsule
   length=16.2002
   radius=9.0458
   shapeName=foot_l_capsule
   center=<-6.4465 6.1026 0.2633>
   rotation=<rot -0.0785 8.0516 -89.5547>
  }]
 foot_r [{
   shapeType=Capsule
   length=16.5966
   radius=9.8181
   shapeName=foot_r_capsule
   center=<7.9102 -4.5962 0.2112>
   rotation=<rot -0.143 9.0524 -89.1891>
  }]
 root [{shapeType=Capsule, length=15.875, radius=984.125, shapeName=root_capsule, center=<0 0 -1002.2814>, rotation=<rot 0 -90 0>}]
 thigh_l [{shapeType=Capsule, length=32.75, radius=9.25, shapeName=thigh_r_capsule, center=<-22.1772 -3 0>, rotation=<rot 90 0 0>}]
 thigh_r [{shapeType=Capsule, length=30.25, radius=8.75, shapeName=thigh_r_capsule, center=<22.1772 3 0>, rotation=<rot 90 0 0>}]
}
constraints [
 {bone1Name=dyn_skirt_l_05, bone2Name=dyn_skirt_l_04, swing1LimitDegrees=35, swing2LimitDegrees=35, twistLimitDegrees=5}
 {bone1Name=dyn_skirt_l_06, bone2Name=dyn_skirt_l_05, swing1LimitDegrees=35, swing2LimitDegrees=35, twistLimitDegrees=5}
 {bone1Name=dyn_skirt_l_04, bone2Name=dyn_skirt_l_03, swing1LimitDegrees=35, swing2LimitDegrees=35, twistLimitDegrees=5}
 {bone1Name=dyn_skirt_l_03, bone2Name=dyn_skirt_l_02, swing1LimitDegrees=35, swing2LimitDegrees=35, twistLimitDegrees=5}
 {bone1Name=dyn_skirt_bk_l_03, bone2Name=dyn_skirt_bk_l_02, swing1LimitDegrees=35, swing2LimitDegrees=35, twistLimitDegrees=5}
 {bone1Name=dyn_skirt_bk_l_04, bone2Name=dyn_skirt_bk_l_03, swing1LimitDegrees=35, swing2LimitDegrees=35, twistLimitDegrees=5}
 {bone1Name=dyn_skirt_bk_l_05, bone2Name=dyn_skirt_bk_l_04, swing1LimitDegrees=35, swing2LimitDegrees=35, twistLimitDegrees=5}
 {
  bone1Name=dyn_skirtH_front_l_03
  bone2Name=dyn_skirtH_front_l_02
  swing1LimitDegrees=20
  swing1Motion=Free
  swing2LimitDegrees=90
  swing2Motion=Free
  twistLimitDegrees=0
 }
 {bone1Name=dyn_skirtH_front_l_04, bone2Name=dyn_skirtH_front_l_03, swing1LimitDegrees=35, swing2LimitDegrees=35, twistLimitDegrees=5}
 {bone1Name=dyn_skirtH_front_l_05, bone2Name=dyn_skirtH_front_l_04, swing1LimitDegrees=35, swing2LimitDegrees=90, twistLimitDegrees=5}
 {bone1Name=dyn_skirtH_front_l_06, bone2Name=dyn_skirtH_front_l_05, swing1LimitDegrees=35, swing2LimitDegrees=90, twistLimitDegrees=5}
 {bone1Name=dyn_skirt_bk_r_05, bone2Name=dyn_skirt_bk_r_04, swing1LimitDegrees=35, swing2LimitDegrees=35, twistLimitDegrees=5}
 {bone1Name=dyn_skirt_bk_r_04, bone2Name=dyn_skirt_bk_r_03, swing1LimitDegrees=35, swing2LimitDegrees=35, twistLimitDegrees=5}
 {bone1Name=dyn_skirt_bk_r_03, bone2Name=dyn_skirt_bk_r_02, swing1LimitDegrees=35, swing2LimitDegrees=35, twistLimitDegrees=5}
 {bone1Name=dyn_skirt_r_06, bone2Name=dyn_skirt_r_05, swing1LimitDegrees=35, swing2LimitDegrees=35, twistLimitDegrees=5}
 {bone1Name=dyn_skirt_r_05, bone2Name=dyn_skirt_r_04, swing1LimitDegrees=35, swing2LimitDegrees=35, twistLimitDegrees=5}
 {bone1Name=dyn_skirt_r_04, bone2Name=dyn_skirt_r_03, swing1LimitDegrees=35, swing2LimitDegrees=35, twistLimitDegrees=5}
 {bone1Name=dyn_skirt_r_03, bone2Name=dyn_skirt_r_02, swing1LimitDegrees=35, swing2LimitDegrees=35, twistLimitDegrees=5}
 {bone1Name=dyn_skirt_front_r_06, bone2Name=dyn_skirt_front_r_05, swing1LimitDegrees=35, swing2LimitDegrees=90, twistLimitDegrees=5}
 {bone1Name=dyn_skirt_front_r_05, bone2Name=dyn_skirt_front_r_04, swing1LimitDegrees=35, swing2LimitDegrees=90, twistLimitDegrees=5}
 {bone1Name=dyn_skirt_front_r_04, bone2Name=dyn_skirt_front_r_03, swing1LimitDegrees=35, swing2LimitDegrees=35, twistLimitDegrees=5}
 {bone1Name=dyn_skirt_front_r_03, bone2Name=dyn_skirt_front_r_02, swing1LimitDegrees=35, swing2LimitDegrees=35, twistLimitDegrees=5}
 {bone1Name=dyn_skirt_bk_l_06, bone2Name=dyn_skirt_bk_l_05, swing1LimitDegrees=35, swing2LimitDegrees=35, twistLimitDegrees=5}
 {bone1Name=dyn_skirt_bk_r_06, bone2Name=dyn_skirt_bk_r_05, swing1LimitDegrees=35, swing2LimitDegrees=35, twistLimitDegrees=5}
 {
  bone1Name=dyn_skirt_l_02
  bone2Name=thigh_l
  swing1LimitDegrees=45
  swing1Motion=Free
  swing2LimitDegrees=45
  swing2Motion=Free
  twistLimitDegrees=45
  twistMotion=Free
 }
 {
  bone1Name=dyn_skirt_bk_l_02
  bone2Name=thigh_l
  swing1LimitDegrees=45
  swing1Motion=Free
  swing2LimitDegrees=45
  swing2Motion=Free
  twistLimitDegrees=45
  twistMotion=Free
 }
 {
  bone1Name=dyn_skirtH_front_l_02
  bone2Name=thigh_l
  swing1LimitDegrees=45
  swing1Motion=Free
  swing2LimitDegrees=45
  swing2Motion=Free
  twistLimitDegrees=45
  twistMotion=Free
 }
 {
  bone1Name=calf_l
  bone2Name=thigh_l
  swing1LimitDegrees=45
  swing1Motion=Free
  swing2LimitDegrees=45
  swing2Motion=Free
  twistLimitDegrees=45
  twistMotion=Free
 }
 {
  bone1Name=dyn_skirt_front_r_02
  bone2Name=thigh_r
  swing1LimitDegrees=45
  swing1Motion=Free
  swing2LimitDegrees=45
  swing2Motion=Free
  twistLimitDegrees=45
  twistMotion=Free
 }
 {
  bone1Name=dyn_skirt_r_02
  bone2Name=thigh_r
  swing1LimitDegrees=45
  swing1Motion=Free
  swing2LimitDegrees=45
  swing2Motion=Free
  twistLimitDegrees=45
  twistMotion=Free
 }
 {
  bone1Name=dyn_skirt_bk_r_02
  bone2Name=thigh_r
  swing1LimitDegrees=45
  swing1Motion=Free
  swing2LimitDegrees=45
  swing2Motion=Free
  twistLimitDegrees=45
  twistMotion=Free
 }
 {
  bone1Name=calf_r
  bone2Name=thigh_r
  swing1LimitDegrees=45
  swing1Motion=Free
  swing2LimitDegrees=45
  swing2Motion=Free
  twistLimitDegrees=45
  twistMotion=Free
 }
 {
  bone1Name=thigh_l
  bone2Name=root
  swing1LimitDegrees=45
  swing1Motion=Free
  swing2LimitDegrees=45
  swing2Motion=Free
  twistLimitDegrees=45
  twistMotion=Free
 }
 {
  bone1Name=thigh_r
  bone2Name=root
  swing1LimitDegrees=45
  swing1Motion=Free
  swing2LimitDegrees=45
  swing2Motion=Free
  twistLimitDegrees=45
  twistMotion=Free
 }
 {
  bone1Name=foot_l
  bone2Name=calf_l
  swing1LimitDegrees=45
  swing1Motion=Free
  swing2LimitDegrees=45
  swing2Motion=Free
  twistLimitDegrees=45
  twistMotion=Free
 }
 {
  bone1Name=foot_r
  bone2Name=calf_r
  swing1LimitDegrees=45
  swing1Motion=Free
  swing2LimitDegrees=45
  swing2Motion=Free
  twistLimitDegrees=45
  twistMotion=Free
 }
]
