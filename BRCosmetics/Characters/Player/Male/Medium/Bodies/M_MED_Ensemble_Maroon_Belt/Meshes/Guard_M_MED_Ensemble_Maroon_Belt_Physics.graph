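# Physics asset for the Guard mesh of the male medium Ensemble Maroon Belt body cosmetic.
# It chains three capsule bodies, from the small guard1_MSTR capsule through guard1_dyn to guard2_dyn.
# Two constraints link each dyn bone to the one before it, with tight swing limits and zero twist, so the guard bends mainly along one swing axis.
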PhysicsAsset /BRCosmetics/Characters/Player/Male/Medium/Bodies/M_MED_Ensemble_Maroon_Belt/Meshes/Guard_M_MED_Ensemble_Maroon_Belt_Physics
bodies {
 guard1_MSTR [{shapeType=Capsule, length=0.505, radius=0.505, shapeName=guard1_MSTR_capsule, rotation=<rot 0 -90 0>}]
 guard1_dyn [{shapeType=Capsule, length=5, radius=5, shapeName=guard1_dyn_capsule, center=<5 0 0>, rotation=<rot 90 0 0>}]
 guard2_dyn [{shapeType=Capsule, length=5, radius=5, shapeName=guard2_dyn_capsule, center=<5 0 0>, rotation=<rot 90 0 0>}]
}
constraints [
 {bone1Name=guard1_dyn, bone2Name=guard1_MSTR, swing1LimitDegrees=10, swing2LimitDegrees=1, twistLimitDegrees=0}
 {bone1Name=guard2_dyn, bone2Name=guard1_dyn, swing1LimitDegrees=10, swing2LimitDegrees=1, twistLimitDegrees=0}
]
