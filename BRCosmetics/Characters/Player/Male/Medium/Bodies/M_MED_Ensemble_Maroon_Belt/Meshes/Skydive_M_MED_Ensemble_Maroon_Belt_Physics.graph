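# Physics asset for the skydive mesh of the Maroon Belt medium male body cosmetic, simulating its skirt against the legs.
# Six skirt bone chains (front, side and back on each side) use five capsules each, starting from tiny root capsules, alongside thigh and calf capsules.
# Links within chains have narrow swing limits, tighter on the back, and zero twist.
# Chain roots and calves attach to their same-side thigh with free angular motion.
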PhysicsAsset /BRCosmetics/Characters/Player/Male/Medium/Bodies/M_MED_Ensemble_Maroon_Belt/Meshes/Skydive_M_MED_Ensemble_Maroon_Belt_Physics
bodies {
 calf_l [{shapeType=Capsule, length=33.25, radius=13.75, shapeName=calf_r_capsule, center=<-26.4668 -2.0022 0>, rotation=<rot 90 23.6294 28.6295>}]
 calf_r [{shapeType=Capsule, length=35.75, radius=15.25, shapeName=calf_r_capsule, center=<28.6541 0 0>, rotation=<rot 90 0 0>}]
 dyn_skirtH_front_l_02 [{shapeType=Capsule, length=1, radius=1, shapeName=dyn_skirtH_front_l_02_capsule, center=<5 0 10>, rotation=<rot 0 0 90>}]
 dyn_skirtH_front_l_03 [{shapeType=Capsule, length=13.5, radius=6.25, shapeName=dyn_skirtH_front_l_03_capsule, center=<5 0 5>, rotation=<rot 0 0 90>}]
 dyn_skirtH_front_l_04 [{shapeType=Capsule, length=13.25, radius=6.5, shapeName=dyn_skirtH_front_l_04_capsule, center=<5 0 5>, rotation=<rot 0 0 90>}]
 dyn_skirtH_front_l_05 [{shapeType=Capsule, length=10.5, radius=9.25, shapeName=dyn_skirtH_front_l_05_capsule, center=<5 0 5>, rotation=<rot 0 0 90>}]
 dyn_skirtH_front_l_06 [{shapeType=Capsule, length=10.25, radius=9.5, shapeName=dyn_skirtH_front_l_06_capsule, center=<5 0 5>, rotation=<rot 0 0 90>}]
 dyn_skirt_bk_l_02 [{shapeType=Capsule, length=1, radius=1, shapeName=dyn_skirt_bk_l_02_capsule, center=<5 0 10>, rotation=<rot 0 0 90>}]
 dyn_skirt_bk_l_03 [{shapeType=Capsule, length=26, radius=7, shapeName=dyn_skirt_bk_l_03_capsule, center=<5 0 5>, rotation=<rot 0 0 90>}]
 dyn_skirt_bk_l_04 [{shapeType=Capsule, length=24.25, radius=8.75, shapeName=dyn_skirt_bk_l_04_capsule, center=<5 0 5>, rotation=<rot 0 0 90>}]
 dyn_skirt_bk_l_05 [{shapeType=Capsule, length=23.25, radius=9.75, shapeName=dyn_skirt_bk_l_05_capsule, center=<5 0 5>, rotation=<rot 0 0 90>}]
 dyn_skirt_bk_l_06 [{shapeType=Capsule, length=22.75, radius=10.25, shapeName=dyn_skirt_l_06_capsule, center=<5 0 5>, rotation=<rot 0 0 90>}]
 dyn_skirt_bk_r_02 [{shapeType=Capsule, length=1, radius=1, shapeName=dyn_skirt_bk_r_02_capsule, center=<-5 0 -10>, rotation=<rot 0 0 90>}]
 dyn_skirt_bk_r_03 [{shapeType=Capsule, length=29.25, radius=7, shapeName=dyn_skirt_bk_r_03_capsule, center=<-5 0 -5>, rotation=<rot 0 0 90>}]
 dyn_skirt_bk_r_04 [{shapeType=Capsule, length=29.25, radius=7, shapeName=dyn_skirt_bk_r_04_capsule, center=<-5 0 -5>, rotation=<rot 0 0 90>}]
 dyn_skirt_bk_r_05 [{shapeType=Capsule, length=32, radius=4.25, shapeName=dyn_skirt_bk_r_05_capsule, center=<-5 0 -5>, rotation=<rot 0 0 90>}]
 dyn_skirt_bk_r_06 [{shapeType=Capsule, length=32.75, radius=3.5, shapeName=dyn_skirt_front_r_06_capsule, center=<-5 0 -5>, rotation=<rot 0 0 90>}]
 dyn_skirt_front_r_02 [{shapeType=Capsule, length=1, radius=1, shapeName=dyn_skirt_front_r_02_capsule, center=<-5 0 -10>, rotation=<rot 0 0 90>}]
 dyn_skirt_front_r_03 [{shapeType=Capsule, length=10.75, radius=7.75, shapeName=dyn_skirt_front_r_03_capsule, center=<-5 0 -5>, rotation=<rot 0 0 90>}]
 dyn_skirt_front_r_04 [{shapeType=Capsule, length=10.75, radius=9.5, shapeName=dyn_skirt_front_r_04_capsule, center=<-5 0 -5>, rotation=<rot 0 0 90>}]
 dyn_skirt_front_r_05 [{shapeType=Capsule, length=10.75, radius=10, shapeName=dyn_skirt_front_r_05_capsule, center=<-5 0 -5>, rotation=<rot 0 0 90>}]
 dyn_skirt_front_r_06 [{shapeType=Capsule, length=10.75, radius=12.25, shapeName=dyn_skirt_front_r_06_capsule, center=<-5 0 -5>, rotation=<rot 0 0 90>}]
 dyn_skirt_l_02 [{shapeType=Capsule, length=1, radius=1, shapeName=dyn_skirt_l_02_capsule, center=<5 0 10>, rotation=<rot 0 0 90>}]
 dyn_skirt_l_03 [{shapeType=Capsule, length=41.25, radius=4.75, shapeName=dyn_skirt_l_03_capsule, center=<5 0 5>, rotation=<rot 0 0 90>}]
 dyn_skirt_l_04 [{shapeType=Capsule, length=40.5, radius=5.5, shapeName=dyn_skirt_l_04_capsule, center=<5 0 5>, rotation=<rot 0 0 90>}]
 dyn_skirt_l_05 [{shapeType=Capsule, length=38, radius=8, shapeName=dyn_skirt_l_05_capsule, center=<5 0 5>, rotation=<rot 0 0 90>}]
 dyn_skirt_l_06 [{shapeType=Capsule, length=37, radius=9, shapeName=dyn_skirt_l_06_capsule, center=<5 0 5>, rotation=<rot 0 0 90>}]
 dyn_skirt_r_02 [{shapeType=Capsule, length=1, radius=1, shapeName=dyn_skirt_r_02_capsule, center=<-5 0 -10>, rotation=<rot 0 0 90>}]
 dyn_skirt_r_03 [{shapeType=Capsule, length=36.25, radius=4.75, shapeName=dyn_skirt_r_03_capsule, center=<-5 0 -5>, rotation=<rot 0 0 90>}]
 dyn_skirt_r_04 [{shapeType=Capsule, length=34.75, radius=6.25, shapeName=dyn_skirt_r_04_capsule, center=<-5 0 -5>, rotation=<rot 0 0 90>}]
 dyn_skirt_r_05 [{shapeType=Capsule, length=33.75, radius=7.25, shapeName=dyn_skirt_r_05_capsule, center=<-5 0 -5>, rotation=<rot 0 0 90>}]
 dyn_skirt_r_06 [{shapeType=Capsule, length=33, radius=8, shapeName=dyn_skirt_r_06_capsule, center=<-5 0 -5>, rotation=<rot 0 0 90>}]
 thigh_l [{shapeType=Capsule, length=32.75, radius=9.25, shapeName=thigh_r_capsule, center=<-22.1772 -3 0>, rotation=<rot 90 0 0>}]
 thigh_r [{shapeType=Capsule, length=30.25, radius=8.75, shapeName=thigh_r_capsule, center=<22.1772 3 0>, rotation=<rot 90 0 0>}]
}
constraints [
 {bone1Name=dyn_skirt_l_05, bone2Name=dyn_skirt_l_04, swing1LimitDegrees=20, swing2LimitDegrees=20, twistLimitDegrees=0}
 {bone1Name=dyn_skirt_l_06, bone2Name=dyn_skirt_l_05, swing1LimitDegrees=20, swing2LimitDegrees=20, twistLimitDegrees=0}
 {bone1Name=dyn_skirt_l_04, bone2Name=dyn_skirt_l_03, swing1LimitDegrees=20, swing2LimitDegrees=20, twistLimitDegrees=0}
 {bone1Name=dyn_skirt_l_03, bone2Name=dyn_skirt_l_02, swing1LimitDegrees=20, swing2LimitDegrees=20, twistLimitDegrees=0}
 {bone1Name=dyn_skirt_bk_l_03, bone2Name=dyn_skirt_bk_l_02, swing1LimitDegrees=10, swing2LimitDegrees=20, twistLimitDegrees=0}
 {bone1Name=dyn_skirt_bk_l_04, bone2Name=dyn_skirt_bk_l_03, swing1LimitDegrees=10, swing2LimitDegrees=20, twistLimitDegrees=0}
 {bone1Name=dyn_skirt_bk_l_05, bone2Name=dyn_skirt_bk_l_04, swing1LimitDegrees=10, swing2LimitDegrees=20, twistLimitDegrees=0}
 {bone1Name=dyn_skirtH_front_l_03, bone2Name=dyn_skirtH_front_l_02, swing1LimitDegrees=20, swing2LimitDegrees=20, twistLimitDegrees=0}
 {bone1Name=dyn_skirtH_front_l_04, bone2Name=dyn_skirtH_front_l_03, swing1LimitDegrees=20, swing2LimitDegrees=20, twistLimitDegrees=0}
 {bone1Name=dyn_skirtH_front_l_05, bone2Name=dyn_skirtH_front_l_04, swing1LimitDegrees=20, swing2LimitDegrees=20, twistLimitDegrees=0}
 {bone1Name=dyn_skirtH_front_l_06, bone2Name=dyn_skirtH_front_l_05, swing1LimitDegrees=20, swing2LimitDegrees=20, twistLimitDegrees=0}
 {bone1Name=dyn_skirt_bk_r_05, bone2Name=dyn_skirt_bk_r_04, swing1LimitDegrees=10, swing2LimitDegrees=20, twistLimitDegrees=0}
 {bone1Name=dyn_skirt_bk_r_04, bone2Name=dyn_skirt_bk_r_03, swing1LimitDegrees=10, swing2LimitDegrees=20, twistLimitDegrees=0}
 {bone1Name=dyn_skirt_bk_r_03, bone2Name=dyn_skirt_bk_r_02, swing1LimitDegrees=10, swing2LimitDegrees=20, twistLimitDegrees=0}
 {bone1Name=dyn_skirt_r_06, bone2Name=dyn_skirt_r_05, swing1LimitDegrees=20, swing2LimitDegrees=20, twistLimitDegrees=0}
 {bone1Name=dyn_skirt_r_05, bone2Name=dyn_skirt_r_04, swing1LimitDegrees=20, swing2LimitDegrees=20, twistLimitDegrees=0}
 {bone1Name=dyn_skirt_r_04, bone2Name=dyn_skirt_r_03, swing1LimitDegrees=20, swing2LimitDegrees=20, twistLimitDegrees=0}
 {bone1Name=dyn_skirt_r_03, bone2Name=dyn_skirt_r_02, swing1LimitDegrees=20, swing2LimitDegrees=20, twistLimitDegrees=0}
 {bone1Name=dyn_skirt_front_r_06, bone2Name=dyn_skirt_front_r_05, swing1LimitDegrees=20, swing2LimitDegrees=20, twistLimitDegrees=0}
 {bone1Name=dyn_skirt_front_r_05, bone2Name=dyn_skirt_front_r_04, swing1LimitDegrees=20, swing2LimitDegrees=20, twistLimitDegrees=0}
 {bone1Name=dyn_skirt_front_r_04, bone2Name=dyn_skirt_front_r_03, swing1LimitDegrees=20, swing2LimitDegrees=20, twistLimitDegrees=0}
 {bone1Name=dyn_skirt_front_r_03, bone2Name=dyn_skirt_front_r_02, swing1LimitDegrees=20, swing2LimitDegrees=20, twistLimitDegrees=0}
 {bone1Name=dyn_skirt_bk_l_06, bone2Name=dyn_skirt_bk_l_05, swing1LimitDegrees=10, swing2LimitDegrees=20, twistLimitDegrees=0}
 {bone1Name=dyn_skirt_bk_r_06, bone2Name=dyn_skirt_bk_r_05, swing1LimitDegrees=10, swing2LimitDegrees=20, twistLimitDegrees=0}
 {
  bone1Name=dyn_skirt_l_02
  bone2Name=thigh_l
  swing1LimitDegrees=45
  swing1Motion=Free
  swing2LimitDegrees=45
  swing2Motion=Free
  twistLimitDegrees=45
  twistMotion=Free
 }
 {
  bone1Name=dyn_skirt_bk_l_02
  bone2Name=thigh_l
  swing1LimitDegrees=45
  swing1Motion=Free
  swing2LimitDegrees=45
  swing2Motion=Free
  twistLimitDegrees=45
  twistMotion=Free
 }
 {
  bone1Name=dyn_skirtH_front_l_02
  bone2Name=thigh_l
  swing1LimitDegrees=45
  swing1Motion=Free
  swing2LimitDegrees=45
  swing2Motion=Free
  twistLimitDegrees=45
  twistMotion=Free
 }
 {
  bone1Name=calf_l
  bone2Name=thigh_l
  swing1LimitDegrees=45
  swing1Motion=Free
  swing2LimitDegrees=45
  swing2Motion=Free
  twistLimitDegrees=45
  twistMotion=Free
 }
 {
  bone1Name=dyn_skirt_front_r_02
  bone2Name=thigh_r
  swing1LimitDegrees=45
  swing1Motion=Free
  swing2LimitDegrees=45
  swing2Motion=Free
  twistLimitDegrees=45
  twistMotion=Free
 }
 {
  bone1Name=dyn_skirt_r_02
  bone2Name=thigh_r
  swing1LimitDegrees=45
  swing1Motion=Free
  swing2LimitDegrees=45
  swing2Motion=Free
  twistLimitDegrees=45
  twistMotion=Free
 }
 {
  bone1Name=dyn_skirt_bk_r_02
  bone2Name=thigh_r
  swing1LimitDegrees=45
  swing1Motion=Free
  swing2LimitDegrees=45
  swing2Motion=Free
  twistLimitDegrees=45
  twistMotion=Free
 }
 {
  bone1Name=calf_r
  bone2Name=thigh_r
  swing1LimitDegrees=45
  swing1Motion=Free
  swing2LimitDegrees=45
  swing2Motion=Free
  twistLimitDegrees=45
  twistMotion=Free
 }
]
