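# Physics asset for the sliding variant of the medium male player's maroon belt ensemble, simulating its skirt.
# Six five-bone capsule chains cover front, side and back on both sides, sharing swing and twist limits at most joints.
# Left chains anchor to a thigh capsule and right chains to root, with free rotation at thigh-to-root and left-side-chain-to-thigh joints.
# A huge zero-length root capsule sits below the origin, apparently as a floor collider.
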PhysicsAsset /Game/Characters/Player/Male/Medium/Bodies/M_MED_Ensemble_Maroon_Belt/Meshes/Sliding_M_MED_Ensemble_Maroon_Belt_Physics3
bodies {
 dyn_skirtH_front_l_02 [{shapeType=Capsule, length=1, radius=1, shapeName=dyn_skirtH_front_l_02_capsule, center=<5 0 10>, rotation=<rot 0 0 90>}]
 dyn_skirtH_front_l_03 [{shapeType=Capsule, length=0.5, radius=9.5, shapeName=dyn_skirtH_front_l_03_capsule, center=<5 0 9.5>, rotation=<rot 0 0 90>}]
 dyn_skirtH_front_l_04 [{shapeType=Capsule, length=0.5, radius=9.5, shapeName=dyn_skirtH_front_l_04_capsule, center=<5 0 9.5>, rotation=<rot 0 0 90>}]
 dyn_skirtH_front_l_05 [{shapeType=Capsule, length=0, radius=10, shapeName=dyn_skirtH_front_l_05_capsule, center=<5 0 5>, rotation=<rot 0 0 90>}]
 dyn_skirtH_front_l_06 [{shapeType=Capsule, length=0, radius=10, shapeName=dyn_skirtH_front_l_06_capsule, center=<5 0 5>, rotation=<rot 0 0 90>}]
 dyn_skirt_bk_l_02 [{shapeType=Capsule, length=1, radius=1, shapeName=dyn_skirt_bk_l_02_capsule, center=<5 0 10>, rotation=<rot 0 0 90>}]
 dyn_skirt_bk_l_03 [{shapeType=Capsule, length=0, radius=10, shapeName=dyn_skirt_bk_l_03_capsule, center=<5 0 10>, rotation=<rot 0 0 90>}]
 dyn_skirt_bk_l_04 [{shapeType=Capsule, length=0, radius=10, shapeName=dyn_skirt_bk_l_04_capsule, center=<5 0 5>, rotation=<rot 0 0 90>}]
 dyn_skirt_bk_l_05 [{shapeType=Capsule, length=0, radius=10, shapeName=dyn_skirt_bk_l_05_capsule, center=<5 0 5>, rotation=<rot 0 0 90>}]
 dyn_skirt_bk_l_06 [{shapeType=Capsule, length=0, radius=10, shapeName=dyn_skirt_l_06_capsule, center=<5 0 5>, rotation=<rot 0 0 90>}]
 dyn_skirt_bk_r_02 [{shapeType=Capsule, length=1, radius=1, shapeName=dyn_skirt_bk_r_02_capsule, center=<-5 0 -10>, rotation=<rot 0 0 90>}]
 dyn_skirt_bk_r_03 [{shapeType=Capsule, length=0, radius=10, shapeName=dyn_skirt_bk_r_03_capsule, center=<-5 0 -10>, rotation=<rot 0 0 90>}]
 dyn_skirt_bk_r_04 [{shapeType=Capsule, length=0, radius=10, shapeName=dyn_skirt_bk_r_04_capsule, center=<-5 0 -5>, rotation=<rot 0 0 90>}]
 dyn_skirt_bk_r_05 [{shapeType=Capsule, length=0, radius=10, shapeName=dyn_skirt_bk_r_05_capsule, center=<-5 0 -5>, rotation=<rot 0 0 90>}]
 dyn_skirt_bk_r_06 [{shapeType=Capsule, length=0, radius=10, shapeName=dyn_skirt_front_r_06_capsule, center=<-5 0 -5>, rotation=<rot 0 0 90>}]
 dyn_skirt_front_r_02 [{shapeType=Capsule, length=1, radius=1, shapeName=dyn_skirt_front_r_02_capsule, center=<-5 0 -10>, rotation=<rot 0 0 90>}]
 dyn_skirt_front_r_03 [{shapeType=Capsule, length=0.25, radius=16.75, shapeName=dyn_skirt_front_r_03_capsule, center=<-5 0 -5>, rotation=<rot 0 0 90>}]
 dyn_skirt_front_r_04 [{shapeType=Capsule, length=3.25, radius=16.5, shapeName=dyn_skirt_front_r_04_capsule, center=<-5 0 -5>, rotation=<rot 0 0 90>}]
 dyn_skirt_front_r_05 [{shapeType=Capsule, length=0.25, radius=16.5, shapeName=dyn_skirt_front_r_05_capsule, center=<-5 0 -5>, rotation=<rot 0 0 90>}]
 dyn_skirt_front_r_06 [{shapeType=Capsule, length=0, radius=10, shapeName=dyn_skirt_front_r_06_capsule, center=<-5 0 -5>, rotation=<rot 0 0 90>}]
 dyn_skirt_l_02 [{shapeType=Capsule, length=1, radius=1, shapeName=dyn_skirt_l_02_capsule, center=<5 0 10>, rotation=<rot 0 0 90>}]
 dyn_skirt_l_03 [{shapeType=Capsule, length=0, radius=10, shapeName=dyn_skirt_l_03_capsule, center=<5 0 5>, rotation=<rot 0 0 90>}]
 dyn_skirt_l_04 [{shapeType=Capsule, length=0, radius=10, shapeName=dyn_skirt_l_04_capsule, center=<5 0 5>, rotation=<rot 0 0 90>}]
 dyn_skirt_l_05 [{shapeType=Capsule, length=0, radius=10, shapeName=dyn_skirt_l_05_capsule, center=<5 0 5>, rotation=<rot 0 0 90>}]
 dyn_skirt_l_06 [{shapeType=Capsule, length=0, radius=10, shapeName=dyn_skirt_l_06_capsule, center=<5 0 5>, rotation=<rot 0 0 90>}]
 dyn_skirt_r_02 [{shapeType=Capsule, length=1, radius=1, shapeName=dyn_skirt_r_02_capsule, center=<-5 0 -10>, rotation=<rot 0 0 90>}]
 dyn_skirt_r_03 [{shapeType=Capsule, length=0, radius=10, shapeName=dyn_skirt_r_03_capsule, center=<-5 0 -5>, rotation=<rot 0 0 90>}]
 dyn_skirt_r_04 [{shapeType=Capsule, length=0, radius=10, shapeName=dyn_skirt_r_04_capsule, center=<-5 0 -5>, rotation=<rot 0 0 90>}]
 dyn_skirt_r_05 [{shapeType=Capsule, length=0, radius=10, shapeName=dyn_skirt_r_05_capsule, center=<-5 0 -5>, rotation=<rot 0 0 90>}]
 dyn_skirt_r_06 [{shapeType=Capsule, length=0, radius=10, shapeName=dyn_skirt_r_06_capsule, center=<-5 0 -5>, rotation=<rot 0 0 90>}]
 root [{shapeType=Capsule, length=0, radius=1000, shapeName=root_capsule, center=<0 3.0022 -1000.0189>, rotation=<rot 0 -90 0>}]
 thigh_l [{shapeType=Capsule, length=26.9925, radius=9.0841, shapeName=thigh_l_capsule, center=<-20.5545 0.0002 0>, rotation=<rot 90 179.7624 179.7629>}]
}
constraints [
 {bone1Name=dyn_skirt_l_05, bone2Name=dyn_skirt_l_04, swing1LimitDegrees=20, swing2LimitDegrees=90, twistLimitDegrees=0}
 {bone1Name=dyn_skirt_l_06, bone2Name=dyn_skirt_l_05, swing1LimitDegrees=20, swing2LimitDegrees=90, twistLimitDegrees=0}
 {bone1Name=dyn_skirt_l_04, bone2Name=dyn_skirt_l_03, swing1LimitDegrees=20, swing2LimitDegrees=90, twistLimitDegrees=0}
 {bone1Name=dyn_skirt_l_03, bone2Name=dyn_skirt_l_02, swing1LimitDegrees=20, swing2LimitDegrees=90, twistLimitDegrees=0}
 {bone1Name=dyn_skirt_bk_l_03, bone2Name=dyn_skirt_bk_l_02, swing1LimitDegrees=20, swing2LimitDegrees=90, twistLimitDegrees=0}
 {bone1Name=dyn_skirt_bk_l_04, bone2Name=dyn_skirt_bk_l_03, swing1LimitDegrees=20, swing2LimitDegrees=90, twistLimitDegrees=0}
 {bone1Name=dyn_skirt_bk_l_05, bone2Name=dyn_skirt_bk_l_04, swing1LimitDegrees=20, swing2LimitDegrees=90, twistLimitDegrees=0}
 {bone1Name=dyn_skirtH_front_l_03, bone2Name=dyn_skirtH_front_l_02, swing1LimitDegrees=20, swing2LimitDegrees=90, twistLimitDegrees=0}
 {bone1Name=dyn_skirtH_front_l_04, bone2Name=dyn_skirtH_front_l_03, swing1LimitDegrees=20, swing2LimitDegrees=90, twistLimitDegrees=0}
 {bone1Name=dyn_skirtH_front_l_05, bone2Name=dyn_skirtH_front_l_04, swing1LimitDegrees=20, swing2LimitDegrees=90, twistLimitDegrees=0}
 {bone1Name=dyn_skirtH_front_l_06, bone2Name=dyn_skirtH_front_l_05, swing1LimitDegrees=20, swing2LimitDegrees=90, twistLimitDegrees=0}
 {bone1Name=dyn_skirt_bk_r_05, bone2Name=dyn_skirt_bk_r_04, swing1LimitDegrees=20, swing2LimitDegrees=90, twistLimitDegrees=0}
 {bone1Name=dyn_skirt_bk_r_04, bone2Name=dyn_skirt_bk_r_03, swing1LimitDegrees=20, swing2LimitDegrees=90, twistLimitDegrees=0}
 {bone1Name=dyn_skirt_bk_r_03, bone2Name=dyn_skirt_bk_r_02, swing1LimitDegrees=20, swing2LimitDegrees=90, twistLimitDegrees=0}
 {bone1Name=dyn_skirt_r_06, bone2Name=dyn_skirt_r_05, swing1LimitDegrees=20, swing2LimitDegrees=90, twistLimitDegrees=0}
 {bone1Name=dyn_skirt_r_05, bone2Name=dyn_skirt_r_04, swing1LimitDegrees=20, swing2LimitDegrees=90, twistLimitDegrees=0}
 {bone1Name=dyn_skirt_r_04, bone2Name=dyn_skirt_r_03, swing1LimitDegrees=20, swing2LimitDegrees=90, twistLimitDegrees=0}
 {bone1Name=dyn_skirt_r_03, bone2Name=dyn_skirt_r_02, swing1LimitDegrees=20, swing2LimitDegrees=90, twistLimitDegrees=0}
 {bone1Name=dyn_skirt_front_r_06, bone2Name=dyn_skirt_front_r_05, swing1LimitDegrees=20, swing2LimitDegrees=90, twistLimitDegrees=0}
 {bone1Name=dyn_skirt_front_r_05, bone2Name=dyn_skirt_front_r_04, swing1LimitDegrees=20, swing2LimitDegrees=90, twistLimitDegrees=0}
 {bone1Name=dyn_skirt_front_r_04, bone2Name=dyn_skirt_front_r_03, swing1LimitDegrees=20, swing2LimitDegrees=90, twistLimitDegrees=0}
 {bone1Name=dyn_skirt_front_r_03, bone2Name=dyn_skirt_front_r_02, swing1LimitDegrees=20, swing2LimitDegrees=90, twistLimitDegrees=0}
 {bone1Name=dyn_skirt_bk_l_06, bone2Name=dyn_skirt_bk_l_05, swing1LimitDegrees=20, swing2LimitDegrees=90, twistLimitDegrees=0}
 {bone1Name=dyn_skirt_bk_r_06, bone2Name=dyn_skirt_bk_r_05, swing1LimitDegrees=20, swing2LimitDegrees=90, twistLimitDegrees=0}
 {
  bone1Name=dyn_skirt_l_02
  bone2Name=thigh_l
  swing1LimitDegrees=45
  swing1Motion=Free
  swing2LimitDegrees=45
  swing2Motion=Free
  twistLimitDegrees=45
  twistMotion=Free
 }
 {bone1Name=dyn_skirt_bk_l_02, bone2Name=thigh_l, swing1LimitDegrees=20, swing2LimitDegrees=90, twistLimitDegrees=0}
 {bone1Name=dyn_skirtH_front_l_02, bone2Name=thigh_l, swing1LimitDegrees=20, swing2LimitDegrees=90, twistLimitDegrees=0}
 {bone1Name=dyn_skirt_front_r_02, bone2Name=root, swing1LimitDegrees=20, swing2LimitDegrees=90, twistLimitDegrees=0}
 {bone1Name=dyn_skirt_r_02, bone2Name=root, swing1LimitDegrees=20, swing2LimitDegrees=90, twistLimitDegrees=0}
 {bone1Name=dyn_skirt_bk_r_02, bone2Name=root, swing1LimitDegrees=20, swing2LimitDegrees=90, twistLimitDegrees=0}
 {
  bone1Name=thigh_l
  bone2Name=root
  swing1LimitDegrees=45
  swing1Motion=Free
  swing2LimitDegrees=45
  swing2Motion=Free
  twistLimitDegrees=45
  twistMotion=Free
 }
]
