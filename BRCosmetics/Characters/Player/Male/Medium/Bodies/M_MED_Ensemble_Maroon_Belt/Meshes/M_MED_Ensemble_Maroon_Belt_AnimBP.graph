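# Post-process-style animation Blueprint for the Maroon Belt medium male body cosmetic, driving its skirt and guard bones.
# Poses copied from the attached parent mesh pass through a deform Control Rig, then bone constraints or RigidBody simulations.
# RigidBody nodes use separate physics assets for sliding, front end, guard and skydiving, some gated by state variables.
# Additive bone offsets, weighted by crouch, movement and thigh-pose variables, adjust the skirt.
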AnimBlueprint /BRCosmetics/Characters/Player/Male/Medium/Bodies/M_MED_Ensemble_Maroon_Belt/Meshes/M_MED_Ensemble_Maroon_Belt_AnimBP
target_skeleton=&/M_MED_Ensemble_Maroon_Belt_Skeleton
graphs [
 AnimGraph {nodes [
   n1 = Control Rig(Source=Copy Pose From Mesh() {bUseAttachedParent=true}) {
    alphaBoolBlend {blendOption=Linear}
    controlRigAssetReference {blueprintRigClass=&/M_LRG_DeformRig_Ensemble_Maroon_Belt_ControlRig1.M_LRG_DeformRig_Ensemble_Maroon_Belt_ControlRig1_C}
   }
   n2 = "Constraint - guard1_MSTR (PPPPPPPPPPPPPPP)"!(
    ComponentPose=Local To Component(LocalPose=@n1)
    Alpha=1
    ConstraintWeights_0=0.3565
    ConstraintWeights_1:float=0.0009
    ConstraintWeights_2:float=0.001
    ConstraintWeights_3:float=0.0006
    ConstraintWeights_4:float=0.0067
    ConstraintWeights_5:float=0.0005
    ConstraintWeights_6:float=0
    ConstraintWeights_7:float=0.0001
    ConstraintWeights_8:float=0
    ConstraintWeights_9:float=0.0001
    ConstraintWeights_10:float=0.3001
    ConstraintWeights_11:float=0.3283
    ConstraintWeights_12:float=0.0012
    ConstraintWeights_13:float=0.0038
    ConstraintWeights_14:float=0.0003
   ) -> (Pose) {
    alphaBoolBlend {blendOption=Linear}
    boneToModify {boneName=guard1_MSTR}
    constraintSetup [
     {targetBone {boneName=pelvis}}
     {targetBone {boneName=spine_01}}
     {targetBone {boneName=spine_02}}
     {targetBone {boneName=spine_03}}
     {targetBone {boneName=thigh_l}}
     {targetBone {boneName=calf_l}}
     {targetBone {boneName=calf_twist_01_l}}
     {targetBone {boneName=calf_twist_02_l}}
     {targetBone {boneName=deform_knee_out_l}}
     {targetBone {boneName=deform_knee_in_l}}
     {targetBone {boneName=thigh_twist_01_l}}
     {targetBone {boneName=deform_glute_l}}
     {targetBone {boneName=deform_groin_l}}
     {targetBone {boneName=thigh_twist_01_r}}
     {targetBone {boneName=deform_glute_r}}
    ]
    constraintWeights [1,1,1,1,1,1,1,1,1,1,1,1,1,1,1]
   }
   n3 = "Constraint - dyn_skirt_l_MSTR (RR)"!(ComponentPose=@n2, Alpha=1, ConstraintWeights_0=1, ConstraintWeights_1:float=0) -> (Pose) {
    alphaBoolBlend {blendOption=Linear}
    boneToModify {boneName=dyn_skirt_l_MSTR}
    constraintSetup [{transformType=Rotation, targetBone {boneName=pelvis}}, {transformType=Rotation, targetBone {boneName=thigh_r}}]
    constraintWeights [1,1]
   }
   n4 = "Constraint - dyn_skirt_bk_l_MSTR (RR)"!(ComponentPose=@n3, Alpha=1, ConstraintWeights_0=1, ConstraintWeights_1:float=0) -> (Pose) {
    alphaBoolBlend {blendOption=Linear}
    boneToModify {boneName=dyn_skirt_bk_l_MSTR}
    constraintSetup [{transformType=Rotation, targetBone {boneName=pelvis}}, {transformType=Rotation, targetBone {boneName=thigh_r}}]
    constraintWeights [1,1]
   }
   n5 = "Constraint - dyn_skirtH_front_l_MSTR (RR)"!(ComponentPose=@n4, Alpha=1, ConstraintWeights_0=1, ConstraintWeights_1:float=0) -> (Pose) {
    alphaBoolBlend {blendOption=Linear}
    boneToModify {boneName=dyn_skirtH_front_l_MSTR}
    constraintSetup [{transformType=Rotation, targetBone {boneName=pelvis}}, {transformType=Rotation, targetBone {boneName=thigh_r}}]
    constraintWeights [1,1]
   }
   n6 = "Constraint - dyn_skirt_front_r_MSTR (RR)"!(ComponentPose=@n5, Alpha=1, ConstraintWeights_0=1, ConstraintWeights_1:float=0) -> (Pose) {
    alphaBoolBlend {blendOption=Linear}
    boneToModify {boneName=dyn_skirt_front_r_MSTR}
    constraintSetup [{transformType=Rotation, targetBone {boneName=pelvis}}, {transformType=Rotation, targetBone {boneName=thigh_l}}]
    constraintWeights [1,1]
   }
   n7 = "Constraint - dyn_skirt_bk_r_MSTR (RR)"!(ComponentPose=@n6, Alpha=1, ConstraintWeights_0=1, ConstraintWeights_1:float=0) -> (Pose) {
    alphaBoolBlend {blendOption=Linear}
    boneToModify {boneName=dyn_skirt_bk_r_MSTR}
    constraintSetup [{transformType=Rotation, targetBone {boneName=pelvis}}, {transformType=Rotation, targetBone {boneName=thigh_l}}]
    constraintWeights [1,1]
   }
   n8 = "Constraint - dyn_skirt_front_c_MSTR (RR)"!(ComponentPose=@n7, Alpha=1, ConstraintWeights_0=1, ConstraintWeights_1:float=0) -> (Pose) {
    alphaBoolBlend {blendOption=Linear}
    boneToModify {boneName=dyn_skirt_front_c_MSTR}
    constraintSetup [{transformType=Rotation, targetBone {boneName=pelvis}}, {transformType=Rotation, targetBone {boneName=thigh_l}}]
    constraintWeights [1,1]
   }
   n9 = "Transform (Modify) Bone - Bone: dyn_skirt_l_02"(ComponentPose=@n8, Alpha=1, Translation="0,0,0", Rotation="45,0,0", Scale="1,1,1") {
    rotationMode=BMM_Additive
    rotationSpace=BCS_BoneSpace
    scaleMode=BMM_Additive
    translationMode=BMM_Additive
    alphaBoolBlend {blendOption=Linear}
    alphaScaleBiasClamp {bInterpResult=true}
    boneToModify {boneName=dyn_skirt_l_02}
    rotation {pitch=45}
   }
   n10 = "Transform (Modify) Bone - Bone: dyn_skirt_bk_l_02"(ComponentPose=@n9, Alpha=1, Translation="0,0,20", Rotation="100,0,0", Scale="1,1,1") {
    rotationMode=BMM_Additive
    rotationSpace=BCS_BoneSpace
    scaleMode=BMM_Additive
    translationMode=BMM_Additive
    alphaBoolBlend {blendOption=Linear}
    alphaScaleBiasClamp {bInterpResult=true}
    boneToModify {boneName=dyn_skirt_bk_l_02}
    rotation {pitch=100}
    translation {z=20}
   }
   n11 = "Transform (Modify) Bone - Bone: dyn_skirtH_front_l_MSTR"(ComponentPose=@n10, Alpha=1, Translation="0,0,0", Rotation="30.000,0.000,-0.000", Scale="1,1,1") {
    rotationMode=BMM_Additive
    rotationSpace=BCS_BoneSpace
    scaleMode=BMM_Additive
    translationMode=BMM_Additive
    translationSpace=BCS_BoneSpace
    alphaBoolBlend {blendOption=Linear}
    alphaScaleBiasClamp {bInterpResult=true}
    boneToModify {boneName=dyn_skirtH_front_l_MSTR}
    rotation=<rot 30 0 0>
   }
   n12 = "Transform (Modify) Bone - Bone: dyn_skirt_r_02"(ComponentPose=@n11, Alpha=1, Translation="0,0,0", Rotation="90,0,0", Scale="1,1,1") {
    rotationMode=BMM_Additive
    rotationSpace=BCS_BoneSpace
    scaleMode=BMM_Additive
    translationMode=BMM_Additive
    alphaBoolBlend {blendOption=Linear}
    alphaScaleBiasClamp {bInterpResult=true}
    boneToModify {boneName=dyn_skirt_r_02}
    rotation {pitch=90}
   }
   n13 = "Transform (Modify) Bone - Bone: dyn_skirt_bk_r_02"(ComponentPose=@n12, Alpha=1, Translation="0,0,20", Rotation="150,0,0", Scale="1,1,1") {
    rotationMode=BMM_Additive
    rotationSpace=BCS_BoneSpace
    scaleMode=BMM_Additive
    translationMode=BMM_Additive
    alphaBoolBlend {blendOption=Linear}
    alphaScaleBiasClamp {bInterpResult=true}
    boneToModify {boneName=dyn_skirt_bk_r_02}
    rotation {pitch=150}
    translation {z=20}
   }
   n14 = "Transform (Modify) Bone - Bone: dyn_skirt_front_r_MSTR"(ComponentPose=@n13, Alpha=1, Translation="0,0,0", Rotation="45,0,0", Scale="1,1,1") {
    rotationMode=BMM_Additive
    rotationSpace=BCS_BoneSpace
    scaleMode=BMM_Additive
    translationMode=BMM_Additive
    alphaBoolBlend {blendOption=Linear}
    alphaScaleBiasClamp {bInterpResult=true}
    boneToModify {boneName=dyn_skirt_front_r_MSTR}
    rotation {pitch=45}
   }
   n15 = RigidBody!(ComponentPose=@n14, Alpha, OverrideWorldGravity:Vector="0,0,-980", ExternalForce) -> (Pose) {
    bClampLinearTranslationLimitToRefPose=true
    bOverrideWorldGravity=true
    overridePhysicsAsset=/Game/Characters/Player/Male/Medium/Bodies/M_MED_Ensemble_Maroon_Belt/Meshes/Sliding_M_MED_Ensemble_Maroon_Belt_Physics3
    simulationSpace=BaseBoneSpace
    alphaBoolBlend {blendOption=Linear}
    baseBoneRef {boneName=pelvis}
    componentAppliedLinearAccClamp=<3000 3000 3000>
    componentLinearAccScale=<0.4 0.2 0.2>
    componentLinearVelScale=<0.3 0.3 0.3>
    overrideWorldGravity {z=-980}
    simSpaceSettings {
     maxAngularAcceleration=1000
     maxAngularVelocity=3
     maxLinearAcceleration=700
     maxLinearVelocity=300
     velocityScaleZ=0
     externalLinearDragV=<0.2 0.2>
    }
   }
   n16 = Control Rig(Source=Copy Pose From Mesh() {bUseAttachedParent=true}) {
    alphaBoolBlend {blendOption=Linear}
    controlRigAssetReference {blueprintRigClass=&M_LRG_DeformRig_Ensemble_Maroon_Belt_ControlRig1_C}
   }
   n17 = "Constraint - guard1_MSTR (PPPPPPPPPPPPPPP)"(
    ComponentPose=Local To Component(LocalPose=@n16)
    Alpha=1
    ConstraintWeights_0=0.3565
    ConstraintWeights_1=0.0009
    ConstraintWeights_2=0.001
    ConstraintWeights_3=0.0006
    ConstraintWeights_4=0.0067
    ConstraintWeights_5=0.0005
    ConstraintWeights_6=0
    ConstraintWeights_7=0.0001
    ConstraintWeights_8=0
    ConstraintWeights_9=0.0001
    ConstraintWeights_10=0.3001
    ConstraintWeights_11=0.3283
    ConstraintWeights_12=0.0012
    ConstraintWeights_13=0.0038
    ConstraintWeights_14=0.0003
   ) {
    alphaBoolBlend {blendOption=Linear}
    boneToModify {boneName=guard1_MSTR}
    constraintSetup [
     {targetBone {boneName=pelvis}}
     {targetBone {boneName=spine_01}}
     {targetBone {boneName=spine_02}}
     {targetBone {boneName=spine_03}}
     {targetBone {boneName=thigh_l}}
     {targetBone {boneName=calf_l}}
     {targetBone {boneName=calf_twist_01_l}}
     {targetBone {boneName=calf_twist_02_l}}
     {targetBone {boneName=deform_knee_out_l}}
     {targetBone {boneName=deform_knee_in_l}}
     {targetBone {boneName=thigh_twist_01_l}}
     {targetBone {boneName=deform_glute_l}}
     {targetBone {boneName=deform_groin_l}}
     {targetBone {boneName=thigh_twist_01_r}}
     {targetBone {boneName=deform_glute_r}}
    ]
    constraintWeights [1,1,1,1,1,1,1,1,1,1,1,1,1,1,1]
   }
   n18 = "Constraint - dyn_skirt_l_MSTR (RR)"(ComponentPose=@n17, Alpha=1, ConstraintWeights_0=1, ConstraintWeights_1=0) {
    alphaBoolBlend {blendOption=Linear}
    boneToModify {boneName=dyn_skirt_l_MSTR}
    constraintSetup [{transformType=Rotation, targetBone {boneName=pelvis}}, {transformType=Rotation, targetBone {boneName=thigh_r}}]
    constraintWeights [1,1]
   }
   n19 = "Constraint - dyn_skirt_bk_l_MSTR (RR)"(ComponentPose=@n18, Alpha=1, ConstraintWeights_0=1, ConstraintWeights_1=0) {
    alphaBoolBlend {blendOption=Linear}
    boneToModify {boneName=dyn_skirt_bk_l_MSTR}
    constraintSetup [{transformType=Rotation, targetBone {boneName=pelvis}}, {transformType=Rotation, targetBone {boneName=thigh_r}}]
    constraintWeights [1,1]
   }
   n20 = "Constraint - dyn_skirtH_front_l_MSTR (RR)"(ComponentPose=@n19, Alpha=1, ConstraintWeights_0=1, ConstraintWeights_1=0) {
    alphaBoolBlend {blendOption=Linear}
    boneToModify {boneName=dyn_skirtH_front_l_MSTR}
    constraintSetup [{transformType=Rotation, targetBone {boneName=pelvis}}, {transformType=Rotation, targetBone {boneName=thigh_r}}]
    constraintWeights [1,1]
   }
   n21 = "Constraint - dyn_skirt_front_r_MSTR (RR)"(ComponentPose=@n20, Alpha=1, ConstraintWeights_0=1, ConstraintWeights_1=0) {
    alphaBoolBlend {blendOption=Linear}
    boneToModify {boneName=dyn_skirt_front_r_MSTR}
    constraintSetup [{transformType=Rotation, targetBone {boneName=pelvis}}, {transformType=Rotation, targetBone {boneName=thigh_l}}]
    constraintWeights [1,1]
   }
   n22 = "Constraint - dyn_skirt_bk_r_MSTR (RR)"(ComponentPose=@n21, Alpha=1, ConstraintWeights_0=1, ConstraintWeights_1=0) {
    alphaBoolBlend {blendOption=Linear}
    boneToModify {boneName=dyn_skirt_bk_r_MSTR}
    constraintSetup [{transformType=Rotation, targetBone {boneName=pelvis}}, {transformType=Rotation, targetBone {boneName=thigh_l}}]
    constraintWeights [1,1]
   }
   n23 = "Constraint - dyn_skirt_r_MSTR (RR)"!(ComponentPose=@n22, Alpha=1, ConstraintWeights_0=1, ConstraintWeights_1:float=0) -> (Pose) {
    alphaBoolBlend {blendOption=Linear}
    boneToModify {boneName=dyn_skirt_r_MSTR}
    constraintSetup [{transformType=Rotation, targetBone {boneName=pelvis}}, {transformType=Rotation, targetBone {boneName=thigh_l}}]
    constraintWeights [1,1]
   }
   n24 = "Transform (Modify) Bone - Bone: dyn_skirt_l_02"(ComponentPose=@n23, Alpha=1, Translation="0,0,0", Rotation="45,0,0", Scale="1,1,1") {
    rotationMode=BMM_Additive
    rotationSpace=BCS_BoneSpace
    scaleMode=BMM_Additive
    translationMode=BMM_Additive
    alphaBoolBlend {blendOption=Linear}
    alphaScaleBiasClamp {bInterpResult=true}
    boneToModify {boneName=dyn_skirt_l_02}
    rotation {pitch=45}
   }
   n25 = "Transform (Modify) Bone - Bone: dyn_skirt_bk_l_02"(ComponentPose=@n24, Alpha=1, Translation="0,0,20", Rotation="100,0,0", Scale="1,1,1") {
    rotationMode=BMM_Additive
    rotationSpace=BCS_BoneSpace
    scaleMode=BMM_Additive
    translationMode=BMM_Additive
    alphaBoolBlend {blendOption=Linear}
    alphaScaleBiasClamp {bInterpResult=true}
    boneToModify {boneName=dyn_skirt_bk_l_02}
    rotation {pitch=100}
    translation {z=20}
   }
   n26 = "Transform (Modify) Bone - Bone: dyn_skirtH_front_l_02"(ComponentPose=@n25, Alpha=1, Translation="20,0,0", Rotation="45,0,0", Scale="1,1,1") {
    rotationMode=BMM_Additive
    rotationSpace=BCS_BoneSpace
    scaleMode=BMM_Additive
    translationMode=BMM_Additive
    alphaBoolBlend {blendOption=Linear}
    alphaScaleBiasClamp {bInterpResult=true}
    boneToModify {boneName=dyn_skirtH_front_l_02}
    rotation {pitch=45}
    translation {x=20}
   }
   n27 = "Transform (Modify) Bone - Bone: dyn_skirt_r_MSTR"(ComponentPose=@n26, Alpha=1, Translation="0,0,0", Rotation="100,0,0", Scale="1,1,1") {
    rotationMode=BMM_Additive
    rotationSpace=BCS_BoneSpace
    scaleMode=BMM_Additive
    translationMode=BMM_Additive
    alphaBoolBlend {blendOption=Linear}
    alphaScaleBiasClamp {bInterpResult=true}
    boneToModify {boneName=dyn_skirt_r_MSTR}
    rotation {pitch=100}
   }
   n28 = "Transform (Modify) Bone - Bone: dyn_skirt_bk_r_02"(ComponentPose=@n27, Alpha=1, Translation="0,0,20", Rotation="60,100,0", Scale="1,1,1") {
    rotationMode=BMM_Additive
    rotationSpace=BCS_BoneSpace
    scaleMode=BMM_Additive
    translationMode=BMM_Additive
    alphaBoolBlend {blendOption=Linear}
    alphaScaleBiasClamp {bInterpResult=true}
    boneToModify {boneName=dyn_skirt_bk_r_02}
    rotation {pitch=60, yaw=100}
    translation {z=20}
   }
   n29 = "Transform (Modify) Bone - Bone: dyn_skirt_front_r_MSTR"(ComponentPose=@n28, Alpha=1, Translation="0,0,0", Rotation="100,0,0", Scale="1,1,1") {
    rotationMode=BMM_Additive
    rotationSpace=BCS_BoneSpace
    scaleMode=BMM_Additive
    translationMode=BMM_Additive
    alphaBoolBlend {blendOption=Linear}
    alphaScaleBiasClamp {bInterpResult=true}
    boneToModify {boneName=dyn_skirt_front_r_MSTR}
    rotation {pitch=100}
   }
   n30 = RigidBody(ComponentPose=@n29, OverrideWorldGravity="0,0,-980") {
    bClampLinearTranslationLimitToRefPose=true
    bOverrideWorldGravity=true
    overridePhysicsAsset=&Sliding_M_MED_Ensemble_Maroon_Belt_Physics3
    simulationSpace=BaseBoneSpace
    alphaBoolBlend {blendOption=Linear}
    baseBoneRef {boneName=pelvis}
    componentAppliedLinearAccClamp=<3000 3000 3000>
    componentLinearAccScale=<0.4 0.2 0.2>
    componentLinearVelScale=<0.3 0.3 0.3>
    overrideWorldGravity {z=-980}
    simSpaceSettings {
     maxAngularAcceleration=1000
     maxAngularVelocity=3
     maxLinearAcceleration=700
     maxLinearVelocity=300
     velocityScaleZ=0
     externalLinearDragV=<0.2 0.2>
    }
   }
   Blend Poses by bool(
    BlendPose_0=Blend Poses by bool(
     BlendPose_0=Component To Local(ComponentPose=@n30)
     BlendPose_1=Component To Local(ComponentPose=@n15)
     BlendTime_0=0.3
     BlendTime_1=0.3
     bActiveValue=Get KneeStraight() -> (KneeStraight:bool)
    ) {blendType=Linear}
    BlendTime_0=0.3
    BlendTime_1=0.3
    bActiveValue=Get bIsSliding()
   ) {blendType=Linear}
   "Constraint - dyn_skirt_bk_l_MSTR (RR)"(Alpha=Get Is Crouching() -> (Is Crouching:double), ConstraintWeights_0=0.5, ConstraintWeights_1=0.5) {
    alphaBoolBlend {blendOption=Linear}
    boneToModify {boneName=dyn_skirt_bk_l_MSTR}
    constraintSetup [{transformType=Rotation, targetBone {boneName=pelvis}}, {transformType=Rotation, targetBone {boneName=thigh_l}}]
    constraintWeights [1,1]
   }
   n31 = "Constraint - dyn_skirt_bk_r_MSTR (RR)"(Alpha=1, ConstraintWeights_0=0.75, ConstraintWeights_1=0.25) {
    alphaBoolBlend {blendOption=Linear}
    boneToModify {boneName=dyn_skirt_bk_r_MSTR}
    constraintSetup [{transformType=Rotation, targetBone {boneName=pelvis}}, {transformType=Rotation, targetBone {boneName=thigh_r}}]
    constraintWeights [1,1]
   }
   "Constraint - dyn_skirt_bk_l_MSTR (RR)"(ComponentPose=@n31, Alpha=1, ConstraintWeights_0=1, ConstraintWeights_1=0) {
    alphaBoolBlend {blendOption=Linear}
    boneToModify {boneName=dyn_skirt_bk_l_MSTR}
    constraintSetup [{transformType=Rotation, targetBone {boneName=pelvis}}, {transformType=Rotation, targetBone {boneName=thigh_l}}]
    constraintWeights [1,1]
   }
   n32 = Get isNotMovingNotCrouching() -> (isNotMovingNotCrouching:double)
   n33 = "Transform (Modify) Bone - Bone: dyn_skirt_l_02"(Alpha=@n32, Translation="0,0,0", Rotation="3,0,0", Scale="1,1,1") {
    rotationMode=BMM_Additive
    rotationSpace=BCS_BoneSpace
    scaleMode=BMM_Additive
    translationMode=BMM_Additive
    alphaBoolBlend {blendOption=Linear}
    alphaScaleBiasClamp {bInterpResult=true}
    boneToModify {boneName=dyn_skirt_l_02}
    rotation {pitch=3}
   }
   n34 = "Transform (Modify) Bone - Bone: dyn_skirtH_front_l_02"(ComponentPose=@n33, Alpha=@n32, Translation="3,0,0", Rotation="0,-5,0", Scale="1,1,1") {
    rotationMode=BMM_Additive
    rotationSpace=BCS_BoneSpace
    scaleMode=BMM_Additive
    translationMode=BMM_Additive
    alphaBoolBlend {blendOption=Linear}
    alphaScaleBiasClamp {bInterpResult=true}
    boneToModify {boneName=dyn_skirtH_front_l_02}
    rotation {yaw=-5}
    translation {x=3}
   }
   n35 = Get ThighSideLf() -> (ThighSideLf:double)
   n36 = "Transform (Modify) Bone - Bone: dyn_skirt_l_02"(ComponentPose=@n34, Alpha=@n35, Translation="0,0,0", Rotation="20,0,0", Scale="1,1,1") {
    rotationMode=BMM_Additive
    rotationSpace=BCS_BoneSpace
    scaleMode=BMM_Additive
    translationMode=BMM_Additive
    alphaBoolBlend {blendOption=Linear}
    alphaScaleBiasClamp {bInterpResult=true}
    boneToModify {boneName=dyn_skirt_l_02}
    rotation {pitch=20}
   }
   n37 = "Transform (Modify) Bone - Bone: dyn_skirt_bk_l_02"(ComponentPose=@n36, Alpha=@n35, Translation="0,0,0", Rotation="20,20,0", Scale="1,1,1") {
    rotationMode=BMM_Additive
    rotationSpace=BCS_BoneSpace
    scaleMode=BMM_Additive
    translationMode=BMM_Additive
    alphaBoolBlend {blendOption=Linear}
    alphaScaleBiasClamp {bInterpResult=true}
    boneToModify {boneName=dyn_skirt_bk_l_02}
    rotation {pitch=20, yaw=20}
   }
   n38 = "Transform (Modify) Bone - Bone: dyn_skirtH_front_l_02"(ComponentPose=@n37, Alpha=@n35, Translation="0,0,0", Rotation="20,-20,0", Scale="1,1,1") {
    rotationMode=BMM_Additive
    rotationSpace=BCS_BoneSpace
    scaleMode=BMM_Additive
    translationMode=BMM_Additive
    alphaBoolBlend {blendOption=Linear}
    alphaScaleBiasClamp {bInterpResult=true}
    boneToModify {boneName=dyn_skirtH_front_l_02}
    rotation {pitch=20, yaw=-20}
   }
   n39 = Get ThighSideRt() -> (ThighSideRt:double)
   n40 = "Transform (Modify) Bone - Bone: dyn_skirt_r_02"(ComponentPose=@n38, Alpha=@n39, Translation="0,0,0", Rotation="20,0,0", Scale="1,1,1") {
    rotationMode=BMM_Additive
    rotationSpace=BCS_BoneSpace
    scaleMode=BMM_Additive
    translationMode=BMM_Additive
    alphaBoolBlend {blendOption=Linear}
    alphaScaleBiasClamp {bInterpResult=true}
    boneToModify {boneName=dyn_skirt_r_02}
    rotation {pitch=20}
   }
   n41 = "Transform (Modify) Bone - Bone: dyn_skirt_bk_r_02"(ComponentPose=@n40, Alpha=@n39, Translation="0,0,0", Rotation="20,20,0", Scale="1,1,1") {
    rotationMode=BMM_Additive
    rotationSpace=BCS_BoneSpace
    scaleMode=BMM_Additive
    translationMode=BMM_Additive
    alphaBoolBlend {blendOption=Linear}
    alphaScaleBiasClamp {bInterpResult=true}
    boneToModify {boneName=dyn_skirt_bk_r_02}
    rotation {pitch=20, yaw=20}
   }
   n42 = "Transform (Modify) Bone - Bone: dyn_skirt_front_r_02"(ComponentPose=@n41, Alpha=@n39, Translation="0,0,0", Rotation="20,-20,0", Scale="1,1,1") {
    rotationMode=BMM_Additive
    rotationSpace=BCS_BoneSpace
    scaleMode=BMM_Additive
    translationMode=BMM_Additive
    alphaBoolBlend {blendOption=Linear}
    alphaScaleBiasClamp {bInterpResult=true}
    boneToModify {boneName=dyn_skirt_front_r_02}
    rotation {pitch=20, yaw=-20}
   }
   n43 = Get thigh_up_l() -> (thigh_up_l:double)
   n44 = "Transform (Modify) Bone - Bone: dyn_skirt_l_02"(ComponentPose=@n42, Alpha=@n43, Translation="0,0,0", Rotation="20,0,0", Scale="1,1,1") {
    rotationMode=BMM_Additive
    rotationSpace=BCS_BoneSpace
    scaleMode=BMM_Additive
    translationMode=BMM_Additive
    alphaBoolBlend {blendOption=Linear}
    alphaScaleBiasClamp {bInterpResult=true}
    boneToModify {boneName=dyn_skirt_l_02}
    rotation {pitch=20}
   }
   n45 = "Transform (Modify) Bone - Bone: dyn_skirt_bk_l_02"(ComponentPose=@n44, Alpha=@n43, Translation="0,0,0", Rotation="20,20,0", Scale="1,1,1") {
    rotationMode=BMM_Additive
    rotationSpace=BCS_BoneSpace
    scaleMode=BMM_Additive
    translationMode=BMM_Additive
    alphaBoolBlend {blendOption=Linear}
    alphaScaleBiasClamp {bInterpResult=true}
    boneToModify {boneName=dyn_skirt_bk_l_02}
    rotation {pitch=20, yaw=20}
   }
   n46 = "Transform (Modify) Bone - Bone: dyn_skirtH_front_l_02"(ComponentPose=@n45, Alpha=@n43, Translation="0,0,0", Rotation="20,-20,0", Scale="1,1,1") {
    rotationMode=BMM_Additive
    rotationSpace=BCS_BoneSpace
    scaleMode=BMM_Additive
    translationMode=BMM_Additive
    alphaBoolBlend {blendOption=Linear}
    alphaScaleBiasClamp {bInterpResult=true}
    boneToModify {boneName=dyn_skirtH_front_l_02}
    rotation {pitch=20, yaw=-20}
   }
   n47 = Get thigh_up_r() -> (thigh_up_r:double)
   n48 = "Transform (Modify) Bone - Bone: dyn_skirt_r_02"(ComponentPose=@n46, Alpha=@n47, Translation="0,0,0", Rotation="20,0,0", Scale="1,1,1") {
    rotationMode=BMM_Additive
    rotationSpace=BCS_BoneSpace
    scaleMode=BMM_Additive
    translationMode=BMM_Additive
    alphaBoolBlend {blendOption=Linear}
    alphaScaleBiasClamp {bInterpResult=true}
    boneToModify {boneName=dyn_skirt_r_02}
    rotation {pitch=20}
   }
   n49 = "Transform (Modify) Bone - Bone: dyn_skirt_bk_r_02"(ComponentPose=@n48, Alpha=@n47, Translation="0,0,0", Rotation="20,20,0", Scale="1,1,1") {
    rotationMode=BMM_Additive
    rotationSpace=BCS_BoneSpace
    scaleMode=BMM_Additive
    translationMode=BMM_Additive
    alphaBoolBlend {blendOption=Linear}
    alphaScaleBiasClamp {bInterpResult=true}
    boneToModify {boneName=dyn_skirt_bk_r_02}
    rotation {pitch=20, yaw=20}
   }
   n50 = "Transform (Modify) Bone - Bone: dyn_skirt_front_r_02"(ComponentPose=@n49, Alpha=@n47, Translation="0,0,0", Rotation="20,-20,0", Scale="1,1,1") {
    rotationMode=BMM_Additive
    rotationSpace=BCS_BoneSpace
    scaleMode=BMM_Additive
    translationMode=BMM_Additive
    alphaBoolBlend {blendOption=Linear}
    alphaScaleBiasClamp {bInterpResult=true}
    boneToModify {boneName=dyn_skirt_front_r_02}
    rotation {pitch=20, yaw=-20}
   }
   n51 = "Transform (Modify) Bone - Bone: dyn_skirtH_front_l_02"(
    ComponentPose=@n50
    Alpha=Get isCrouchingNotMoving() -> (isCrouchingNotMoving:double)
    Translation="0,0,0"
    Rotation="-10,20,0"
    Scale="1,1,1"
   ) {
    rotationMode=BMM_Additive
    rotationSpace=BCS_BoneSpace
    scaleMode=BMM_Additive
    translationMode=BMM_Additive
    alphaBoolBlend {blendOption=Linear}
    alphaScaleBiasClamp {bInterpResult=true, interpSpeedIncreasing=5}
    boneToModify {boneName=dyn_skirtH_front_l_02}
    rotation {pitch=-10, yaw=20}
   }
   n52 = Get Is Crouching()
   n53 = "Transform (Modify) Bone - Bone: dyn_skirt_l_02"(ComponentPose=@n51, Alpha=@n52, Translation="0,0,0", Rotation="0,0,0", Scale="1,1,1") {
    rotationMode=BMM_Additive
    rotationSpace=BCS_BoneSpace
    scaleMode=BMM_Additive
    translationMode=BMM_Additive
    alphaBoolBlend {blendOption=Linear}
    boneToModify {boneName=dyn_skirt_l_02}
   }
   n54 = "Transform (Modify) Bone - Bone: dyn_skirt_bk_l_02"(ComponentPose=@n53, Alpha=Get isCrouchingNotMoving(), Translation="0,0,0", Rotation="0,80,20", Scale="1,1,0") {
    rotationMode=BMM_Additive
    rotationSpace=BCS_BoneSpace
    scaleMode=BMM_Additive
    translationMode=BMM_Additive
    alphaBoolBlend {blendOption=Linear}
    alphaScaleBiasClamp {bInterpResult=true, interpSpeedIncreasing=100}
    boneToModify {boneName=dyn_skirt_bk_l_02}
    rotation {roll=20, yaw=80}
    scale {z=0}
   }
   n55 = "Transform (Modify) Bone - Bone: dyn_skirt_r_02"(ComponentPose=@n54, Alpha=@n52, Translation="0,0,0", Rotation="20,20,0", Scale="1,1,1") {
    rotationMode=BMM_Additive
    rotationSpace=BCS_BoneSpace
    scaleMode=BMM_Additive
    translationMode=BMM_Additive
    alphaBoolBlend {blendOption=Linear}
    boneToModify {boneName=dyn_skirt_r_02}
    rotation {pitch=20, yaw=20}
   }
   n56 = "Transform (Modify) Bone - Bone: dyn_skirt_bk_r_MSTR"(ComponentPose=@n55, Alpha=Get isCrouchingNotMoving(), Translation="0,0,0", Rotation="60,60,0", Scale="1,1,1") {
    rotationMode=BMM_Additive
    rotationSpace=BCS_BoneSpace
    scaleMode=BMM_Additive
    translationMode=BMM_Additive
    alphaBoolBlend {blendOption=Linear}
    boneToModify {boneName=dyn_skirt_bk_r_MSTR}
    rotation {pitch=60, yaw=60}
   }
   "Transform (Modify) Bone - Bone: dyn_skirt_bk_r_MSTR"(ComponentPose=@n56, Alpha=Get isCrouchMoving() -> (isCrouchMoving:double), Translation="0,0,0", Rotation="0,30,0", Scale="1,1,1") {
    rotationMode=BMM_Additive
    rotationSpace=BCS_BoneSpace
    scaleMode=BMM_Additive
    translationMode=BMM_Additive
    alphaBoolBlend {blendOption=Linear}
    boneToModify {boneName=dyn_skirt_bk_r_MSTR}
    rotation {yaw=30}
   }
   "Transform (Modify) Bone - Bone: dyn_skirt_bk_l_02"(Alpha=1, Translation="0,0,0", Rotation="30,70,0", Scale="0.5,1,0") {
    rotationMode=BMM_Additive
    rotationSpace=BCS_BoneSpace
    scaleMode=BMM_Additive
    translationMode=BMM_Additive
    alphaBoolBlend {blendOption=Linear}
    boneToModify {boneName=dyn_skirt_bk_l_02}
    rotation {pitch=30, yaw=70}
    scale {x=0.5, z=0}
   }
   n57 = Get RightMovementClean() -> (RightMovementClean:double)
   n58 = "Transform (Modify) Bone - Bone: dyn_skirt_front_r_04"(Alpha=@n57, Translation="0,0,0", Rotation="0,-15,0", Scale="1,1,1") {
    rotationMode=BMM_Additive
    rotationSpace=BCS_BoneSpace
    scaleMode=BMM_Additive
    translationMode=BMM_Additive
    alphaBoolBlend {blendOption=Linear}
    alphaScaleBiasClamp {bInterpResult=true, interpSpeedDecreasing=20, interpSpeedIncreasing=20}
    boneToModify {boneName=dyn_skirt_front_r_04}
    rotation {yaw=-15}
   }
   "Transform (Modify) Bone - Bone: dyn_skirt_front_r_05"(ComponentPose=@n58, Alpha=@n57, Translation="0,0,0", Rotation="0,-15,0", Scale="1,1,1") {
    rotationMode=BMM_Additive
    rotationSpace=BCS_BoneSpace
    scaleMode=BMM_Additive
    translationMode=BMM_Additive
    alphaBoolBlend {blendOption=Linear}
    alphaScaleBiasClamp {bInterpResult=true, interpSpeedDecreasing=20, interpSpeedIncreasing=20}
    boneToModify {boneName=dyn_skirt_front_r_05}
    rotation {yaw=-15}
   }
   "Transform (Modify) Bone - Bone: dyn_skirt_r_MSTR"(Alpha=@n57, Translation="0,0,0", Rotation="0,10,0", Scale="1,1,1") {
    rotationMode=BMM_Additive
    rotationSpace=BCS_BoneSpace
    scaleMode=BMM_Additive
    translationMode=BMM_Additive
    alphaBoolBlend {blendOption=Linear}
    alphaScaleBiasClamp {bInterpResult=true}
    boneToModify {boneName=dyn_skirt_r_MSTR}
    rotation {yaw=10}
   }
   n59 = Get bIsAccelerating2D()
   n60 = "Transform (Modify) Bone - Bone: dyn_skirtH_front_l_MSTR"!(ComponentPose, bAlphaBoolEnabled:bool=@n59, Translation="0,0,0", Rotation="40,0,0", Scale="1,1,1") -> (Pose) {
    alphaInputType=Bool
    rotationMode=BMM_Additive
    rotationSpace=BCS_BoneSpace
    scaleMode=BMM_Additive
    scaleSpace=BCS_BoneSpace
    translationMode=BMM_Additive
    translationSpace=BCS_BoneSpace
    alphaBoolBlend {blendOption=Linear}
    boneToModify {boneName=dyn_skirtH_front_l_MSTR}
    rotation {pitch=40}
   }
   "Transform (Modify) Bone - Bone: dyn_skirt_front_r_MSTR"!(ComponentPose=@n60, bAlphaBoolEnabled:bool=@n59, Translation="0,0,0", Rotation="25,0,0", Scale="1,1,1") -> (Pose) {
    alphaInputType=Bool
    rotationMode=BMM_Additive
    rotationSpace=BCS_BoneSpace
    scaleMode=BMM_Additive
    scaleSpace=BCS_BoneSpace
    translationMode=BMM_Additive
    translationSpace=BCS_BoneSpace
    alphaBoolBlend {blendOption=Linear}
    boneToModify {boneName=dyn_skirt_front_r_MSTR}
    rotation {pitch=25}
   }
   "Transform (Modify) Bone - Bone: dyn_skirtH_front_l_02"!(ComponentPose, bAlphaBoolEnabled:bool=True, Translation="0,0,0", Rotation="0,0,0", Scale="1,1,1") -> (Pose) {
    alphaInputType=Bool
    rotationMode=BMM_Additive
    rotationSpace=BCS_BoneSpace
    scaleMode=BMM_Additive
    translationMode=BMM_Additive
    alphaBoolBlend {blendInTime=0.7, blendOption=Linear, blendOutTime=0.1}
    alphaScaleBiasClamp {bInterpResult=true}
    boneToModify {boneName=dyn_skirtH_front_l_02}
   }
   "Transform (Modify) Bone - Bone: dyn_skirt_front_r_01"!(ComponentPose, bAlphaBoolEnabled:bool=True, Translation="0,0,0", Rotation="0,0,0", Scale="1,1,1") -> (Pose) {
    alphaInputType=Bool
    rotationMode=BMM_Additive
    rotationSpace=BCS_BoneSpace
    scaleMode=BMM_Additive
    translationMode=BMM_Additive
    alphaBoolBlend {blendInTime=0.7, blendOption=Linear, blendOutTime=0.1}
    alphaScaleBiasClamp {bInterpResult=true}
    boneToModify {boneName=dyn_skirt_front_r_01}
   }
   "Transform (Modify) Bone - Bone: dyn_skirtH_front_l_01"(Alpha=@n35, Translation="0,0,0", Rotation="20,0,0", Scale="1,1,1") {
    rotationMode=BMM_Additive
    rotationSpace=BCS_BoneSpace
    scaleMode=BMM_Additive
    translationMode=BMM_Additive
    alphaBoolBlend {blendOption=Linear}
    alphaScaleBiasClamp {bInterpResult=true}
    boneToModify {boneName=dyn_skirtH_front_l_01}
    rotation {pitch=20}
   }
   n61 = Get isNotMovingNotCrouching()
   n62 = "Transform (Modify) Bone - Bone: dyn_skirt_l_03"(Alpha=@n61, Translation="1,0,0", Rotation="3,0,0", Scale="1,1,1") {
    rotationMode=BMM_Additive
    rotationSpace=BCS_BoneSpace
    scaleMode=BMM_Additive
    translationMode=BMM_Additive
    alphaBoolBlend {blendOption=Linear}
    alphaScaleBiasClamp {bInterpResult=true}
    boneToModify {boneName=dyn_skirt_l_03}
    rotation {pitch=3}
    translation {x=1}
   }
   "Transform (Modify) Bone - Bone: dyn_skirtH_front_l_02"!(ComponentPose=@n62, Alpha:float=@n61, Translation="0,0,0", Rotation="5,0,0", Scale="1,1,1") -> (Pose) {
    rotationMode=BMM_Additive
    rotationSpace=BCS_BoneSpace
    scaleMode=BMM_Additive
    translationMode=BMM_Additive
    alphaBoolBlend {blendOption=Linear}
    alphaScaleBiasClamp {bInterpResult=true}
    boneToModify {boneName=dyn_skirtH_front_l_02}
    rotation {pitch=5}
   }
   "Transform (Modify) Bone - Bone: dyn_skirtH_front_l_03"(Alpha=1, Translation="0,0,0", Rotation="120,60,0", Scale="1,1,1") {
    rotationMode=BMM_Additive
    rotationSpace=BCS_BoneSpace
    scaleMode=BMM_Additive
    translationMode=BMM_Additive
    translationSpace=BCS_BoneSpace
    alphaBoolBlend {blendOption=Linear}
    alphaScaleBiasClamp {bInterpResult=true}
    boneToModify {boneName=dyn_skirtH_front_l_03}
    rotation {pitch=120, yaw=60}
   }
   n63 = Control Rig!(
    Source=Copy Pose From Mesh() {bUseAttachedParent=true}
    Alpha
    jumping:double=Get Is Jumping() -> (Is Jumping:double)
    IsRBANEnabled:bool=Get bRigidBodyEnabled()
   ) -> (Pose) {
    alphaBoolBlend {blendOption=Linear}
    controlRigAssetReference {blueprintRigClass=&M_LRG_DeformRig_Ensemble_Maroon_Belt_ControlRig1_C}
   }
   n64 = "Constraint - guard1_MSTR (PPPPPPPPPPPPPPP)"(
    ComponentPose=Local To Component(LocalPose=@n63)
    Alpha=1
    ConstraintWeights_0=0.3565
    ConstraintWeights_1=0.0009
    ConstraintWeights_2=0.001
    ConstraintWeights_3=0.0006
    ConstraintWeights_4=0.0067
    ConstraintWeights_5=0.0005
    ConstraintWeights_6=0
    ConstraintWeights_7=0.0001
    ConstraintWeights_8=0
    ConstraintWeights_9=0.0001
    ConstraintWeights_10=0.3001
    ConstraintWeights_11=0.3283
    ConstraintWeights_12=0.0012
    ConstraintWeights_13=0.0038
    ConstraintWeights_14=0.0003
   ) {
    alphaBoolBlend {blendOption=Linear}
    boneToModify {boneName=guard1_MSTR}
    constraintSetup [
     {targetBone {boneName=pelvis}}
     {targetBone {boneName=spine_01}}
     {targetBone {boneName=spine_02}}
     {targetBone {boneName=spine_03}}
     {targetBone {boneName=thigh_l}}
     {targetBone {boneName=calf_l}}
     {targetBone {boneName=calf_twist_01_l}}
     {targetBone {boneName=calf_twist_02_l}}
     {targetBone {boneName=deform_knee_out_l}}
     {targetBone {boneName=deform_knee_in_l}}
     {targetBone {boneName=thigh_twist_01_l}}
     {targetBone {boneName=deform_glute_l}}
     {targetBone {boneName=deform_groin_l}}
     {targetBone {boneName=thigh_twist_01_r}}
     {targetBone {boneName=deform_glute_r}}
    ]
    constraintWeights [1,1,1,1,1,1,1,1,1,1,1,1,1,1,1]
   }
   n65 = RigidBody!(
    ComponentPose=@n64
    bAlphaBoolEnabled:bool=Get inFrontEndNotEmoting() -> (inFrontEndNotEmoting:bool)
    OverrideWorldGravity="0,0,-300"
    ExternalForce
   ) -> (Pose) {
    alphaInputType=Bool
    bClampLinearTranslationLimitToRefPose=true
    bOverrideWorldGravity=true
    overridePhysicsAsset=&/Frontend_M_MED_Ensemble_Maroon_Belt_Physics2
    simulationSpace=BaseBoneSpace
    alphaBoolBlend {blendOption=Linear}
    baseBoneRef {boneName=root}
    componentAppliedLinearAccClamp=<3000 3000 3000>
    componentLinearAccScale=<0.4 0.2 0.2>
    componentLinearVelScale=<0.3 0.3 0.3>
    overrideWorldGravity {z=-980}
    simSpaceSettings {
     maxAngularAcceleration=1000
     maxAngularVelocity=3
     maxLinearAcceleration=700
     maxLinearVelocity=300
     velocityScaleZ=0
     worldAlpha=0.2
     externalLinearDragV=<0.2 0.2>
    }
   }
   n66 = Get GravityOverride()
   n67 = RigidBody!(ComponentPose=@n65, Alpha:float=Get isNotSkydivingNotInFE() -> (isNotSkydivingNotInFE:double), OverrideWorldGravity=@n66, ExternalForce) -> (Pose) {
    bClampLinearTranslationLimitToRefPose=true
    bOverrideWorldGravity=true
    overridePhysicsAsset=&Frontend_M_MED_Ensemble_Maroon_Belt_Physics2
    simulationSpace=BaseBoneSpace
    alphaBoolBlend {blendOption=Linear}
    baseBoneRef {boneName=root}
    componentAppliedLinearAccClamp=<3000 3000 3000>
    componentLinearAccScale=<0.4 0.2 0.2>
    componentLinearVelScale=<0.3 0.3 0.3>
    overrideWorldGravity {z=-980}
    simSpaceSettings {
     maxAngularAcceleration=1000
     maxAngularVelocity=3
     maxLinearAcceleration=700
     maxLinearVelocity=300
     velocityScaleZ=0
     worldAlpha=0.2
     externalLinearDragV=<0.2 0.2>
    }
   }
   n68 = RigidBody!(
    ComponentPose=@n67
    bAlphaBoolEnabled:bool=Get isInFrontEndIsEmoting() -> (isInFrontEndIsEmoting:bool)
    OverrideWorldGravity=@n66
    ExternalForce
   ) -> (Pose) {
    alphaInputType=Bool
    bClampLinearTranslationLimitToRefPose=true
    bOverrideWorldGravity=true
    overridePhysicsAsset=&Frontend_M_MED_Ensemble_Maroon_Belt_Physics2
    simulationSpace=BaseBoneSpace
    alphaBoolBlend {blendOption=Linear}
    baseBoneRef {boneName=root}
    componentAppliedLinearAccClamp=<3000 3000 3000>
    componentLinearAccScale=<0.4 0.2 0.2>
    componentLinearVelScale=<0.3 0.3 0.3>
    overrideWorldGravity {z=-980}
    simSpaceSettings {
     maxAngularAcceleration=1000
     maxAngularVelocity=3
     maxLinearAcceleration=700
     maxLinearVelocity=300
     velocityScaleZ=0
     worldAlpha=0.2
     externalLinearDragV=<0.2 0.2>
    }
   }
   n69 = RigidBody!(ComponentPose=@n68, Alpha:float=0.2, OverrideWorldGravity=@n66, ExternalForce) -> (Pose) {
    bClampLinearTranslationLimitToRefPose=true
    bOverrideWorldGravity=true
    overridePhysicsAsset=&/Guard_M_MED_Ensemble_Maroon_Belt_Physics
    simulationSpace=BaseBoneSpace
    alphaBoolBlend {blendOption=Linear}
    baseBoneRef {boneName=root}
    componentAppliedLinearAccClamp=<3000 3000 3000>
    componentLinearAccScale=<0.5 0.3 0.3>
    componentLinearVelScale=<0.3 0.3 0.3>
    overrideWorldGravity {z=-980}
    simSpaceSettings {
     maxAngularAcceleration=1000
     maxAngularVelocity=3
     maxLinearAcceleration=700
     maxLinearVelocity=300
     velocityScaleZ=0
     worldAlpha=0.25
     externalLinearDragV=<0.2 0.2>
    }
   }
   n70 = RigidBody(ComponentPose=@n69, Alpha=Get Is Sky Diving() -> (Is Sky Diving:double), OverrideWorldGravity=@n66) {
    bClampLinearTranslationLimitToRefPose=true
    bOverrideWorldGravity=true
    overridePhysicsAsset=&/Skydive_M_MED_Ensemble_Maroon_Belt_Physics
    simulationSpace=BaseBoneSpace
    alphaBoolBlend {blendOption=Linear}
    baseBoneRef {boneName=root}
    componentAppliedLinearAccClamp=<3000 3000 3000>
    componentLinearAccScale=<0.4 0.2 0.2>
    componentLinearVelScale=<0.3 0.3 0.3>
    overrideWorldGravity {z=-980}
    simSpaceSettings {
     maxAngularAcceleration=1000
     maxAngularVelocity=3
     maxLinearAcceleration=700
     maxLinearVelocity=300
     velocityScaleZ=0
     worldAlpha=0.2
     externalLinearDragV=<0.2 0.2>
    }
   }
   n71 = Get isCrouchingNotMoving()
   n72 = "Transform (Modify) Bone - Bone: dyn_skirt_bk_r_MSTR"(ComponentPose=@n70, Alpha=@n71, Translation="8,-2,-2", Rotation="0,0,0", Scale="1,1,1") {
    rotationMode=BMM_Additive
    rotationSpace=BCS_BoneSpace
    scaleMode=BMM_Additive
    scaleSpace=BCS_BoneSpace
    translationMode=BMM_Additive
    translationSpace=BCS_BoneSpace
    alphaBoolBlend {blendInTime=0.1, blendOption=Linear, blendOutTime=0.1}
    alphaScaleBiasClamp {bInterpResult=true, interpSpeedIncreasing=20}
    boneToModify {boneName=dyn_skirt_bk_r_MSTR}
    translation=<8 -2 -2>
   }
   n73 = "Transform (Modify) Bone - Bone: dyn_skirt_r_MSTR"(ComponentPose=@n72, Alpha=@n71, Translation="5,-2,-2", Rotation="0,0,0", Scale="1,1,1") {
    rotationMode=BMM_Additive
    rotationSpace=BCS_BoneSpace
    scaleMode=BMM_Additive
    scaleSpace=BCS_BoneSpace
    translationMode=BMM_Additive
    translationSpace=BCS_BoneSpace
    alphaBoolBlend {blendInTime=0.1, blendOption=Linear, blendOutTime=0.1}
    alphaScaleBiasClamp {bInterpResult=true, interpSpeedIncreasing=20}
    boneToModify {boneName=dyn_skirt_r_MSTR}
    translation=<5 -2 -2>
   }
   n74 = "Transform (Modify) Bone - Bone: dyn_skirt_bk_l_MSTR"(ComponentPose=@n73, Alpha=@n71, Translation="-5,2,1", Rotation="0,0,0", Scale="1,1,1") {
    rotationMode=BMM_Additive
    rotationSpace=BCS_BoneSpace
    scaleMode=BMM_Additive
    scaleSpace=BCS_BoneSpace
    translationMode=BMM_Additive
    translationSpace=BCS_BoneSpace
    alphaBoolBlend {blendInTime=0.1, blendOption=Linear, blendOutTime=0.1}
    alphaScaleBiasClamp {bInterpResult=true, interpSpeedIncreasing=20}
    boneToModify {boneName=dyn_skirt_bk_l_MSTR}
    translation=<-5 2 1>
   }
   n75 = "Transform (Modify) Bone - Bone: dyn_skirt_l_MSTR"(ComponentPose=@n74, Alpha=@n71, Translation="-3,2,0", Rotation="0,0,0", Scale="1,1,1") {
    rotationMode=BMM_Additive
    rotationSpace=BCS_BoneSpace
    scaleMode=BMM_Additive
    scaleSpace=BCS_BoneSpace
    translationMode=BMM_Additive
    translationSpace=BCS_BoneSpace
    alphaBoolBlend {blendInTime=0.1, blendOption=Linear, blendOutTime=0.1}
    alphaScaleBiasClamp {bInterpResult=true, interpSpeedIncreasing=20}
    boneToModify {boneName=dyn_skirt_l_MSTR}
    translation=<-3 2>
   }
   n76 = Get Is Moving 2D() -> (Is Moving 2D:double)
   n77 = "Transform (Modify) Bone - Bone: dyn_skirt_l_01"(ComponentPose=@n75, Alpha=@n76, Translation="0,0,0", Rotation="5,0,0.166", Scale="1,1,1") {
    rotationMode=BMM_Additive
    rotationSpace=BCS_BoneSpace
    scaleMode=BMM_Additive
    translationMode=BMM_Additive
    alphaBoolBlend {blendOption=Linear}
    alphaScaleBiasClamp {bInterpResult=true}
    boneToModify {boneName=dyn_skirt_l_01}
    rotation {pitch=5, roll=0.166}
   }
   n78 = "Transform (Modify) Bone - Bone: dyn_skirt_bk_l_01"(ComponentPose=@n77, Alpha=@n76, Translation="0,0,0", Rotation="7,7,0", Scale="1,1,1") {
    rotationMode=BMM_Additive
    rotationSpace=BCS_BoneSpace
    scaleMode=BMM_Additive
    translationMode=BMM_Additive
    alphaBoolBlend {blendOption=Linear}
    alphaScaleBiasClamp {bInterpResult=true}
    boneToModify {boneName=dyn_skirt_bk_l_01}
    rotation {pitch=7, yaw=7}
   }
   n79 = "Transform (Modify) Bone - Bone: dyn_skirtH_front_l_02"(ComponentPose=@n78, Alpha=@n76, Translation="0,0,0", Rotation="5,-5,0", Scale="1,1,1") {
    rotationMode=BMM_Additive
    rotationSpace=BCS_BoneSpace
    scaleMode=BMM_Additive
    translationMode=BMM_Additive
    alphaBoolBlend {blendOption=Linear}
    alphaScaleBiasClamp {bInterpResult=true}
    boneToModify {boneName=dyn_skirtH_front_l_02}
    rotation {pitch=5, yaw=-5}
   }
   n80 = "Transform (Modify) Bone - Bone: dyn_skirt_r_01"(ComponentPose=@n79, Alpha=@n76, Translation="0,0,0", Rotation="10,0,0", Scale="1,1,1") {
    rotationMode=BMM_Additive
    rotationSpace=BCS_BoneSpace
    scaleMode=BMM_Additive
    translationMode=BMM_Additive
    alphaBoolBlend {blendOption=Linear}
    alphaScaleBiasClamp {bInterpResult=true}
    boneToModify {boneName=dyn_skirt_r_01}
    rotation {pitch=10}
   }
   n81 = "Transform (Modify) Bone - Bone: dyn_skirt_bk_r_02"(ComponentPose=@n80, Alpha=@n76, Translation="0,0,0", Rotation="10,10,0", Scale="1,1,1") {
    rotationMode=BMM_Additive
    rotationSpace=BCS_BoneSpace
    scaleMode=BMM_Additive
    translationMode=BMM_Additive
    alphaBoolBlend {blendOption=Linear}
    alphaScaleBiasClamp {bInterpResult=true}
    boneToModify {boneName=dyn_skirt_bk_r_02}
    rotation {pitch=10, yaw=10}
   }
   n82 = "Transform (Modify) Bone - Bone: dyn_skirt_front_r_02"(ComponentPose=@n81, Alpha=@n76, Translation="0,0,0", Rotation="5,-5,0", Scale="1,1,1") {
    rotationMode=BMM_Additive
    rotationSpace=BCS_BoneSpace
    scaleMode=BMM_Additive
    translationMode=BMM_Additive
    alphaBoolBlend {blendOption=Linear}
    alphaScaleBiasClamp {bInterpResult=true}
    boneToModify {boneName=dyn_skirt_front_r_02}
    rotation {pitch=5, yaw=-5}
   }
   n83 = "Transform (Modify) Bone - Bone: dyn_skirt_front_r_MSTR"!(ComponentPose=@n82, Alpha:float=Get Is Jumping(), Translation="0,0,3", Rotation="10,0,0", Scale="1,1,1") -> (Pose) {
    rotationMode=BMM_Additive
    rotationSpace=BCS_BoneSpace
    scaleMode=BMM_Additive
    translationMode=BMM_Additive
    alphaBoolBlend {blendOption=Linear}
    alphaScaleBiasClamp {bInterpResult=true}
    boneToModify {boneName=dyn_skirt_front_r_MSTR}
    rotation {pitch=10}
    translation {z=3}
   }
   n84 = Get bIsBaseballSliding()
   n85 = "Transform (Modify) Bone - Bone: dyn_skirt_l_01"!(ComponentPose=@n83, bAlphaBoolEnabled:bool=@n84, Translation="0,0,0", Rotation="47.790,-22.480,-16.917", Scale="1,1,1") -> (Pose) {
    alphaInputType=Bool
    rotationMode=BMM_Additive
    rotationSpace=BCS_BoneSpace
    scaleMode=BMM_Additive
    translationMode=BMM_Additive
    alphaBoolBlend {blendInTime=0.7, blendOption=Linear, blendOutTime=0.1}
    alphaScaleBiasClamp {bInterpResult=true}
    boneToModify {boneName=dyn_skirt_l_01}
    rotation=<rot 47.79 -22.48 -16.917>
   }
   n86 = "Transform (Modify) Bone - Bone: dyn_skirt_bk_l_01"!(ComponentPose=@n85, bAlphaBoolEnabled:bool=@n84, Translation="0,0,0", Rotation="88,0,0", Scale="1,1,1") -> (Pose) {
    alphaInputType=Bool
    rotationMode=BMM_Additive
    rotationSpace=BCS_BoneSpace
    scaleMode=BMM_Additive
    translationMode=BMM_Additive
    alphaBoolBlend {blendInTime=0.7, blendOption=Linear, blendOutTime=0.1}
    alphaScaleBiasClamp {bInterpResult=true}
    boneToModify {boneName=dyn_skirt_bk_l_01}
    rotation {pitch=88}
   }
   n87 = "Transform (Modify) Bone - Bone: dyn_skirtH_front_l_02"!(ComponentPose=@n86, bAlphaBoolEnabled:bool=@n84, Translation="0,0,0", Rotation="0,0,0", Scale="1,1,1") -> (Pose) {
    alphaInputType=Bool
    rotationMode=BMM_Additive
    rotationSpace=BCS_BoneSpace
    scaleMode=BMM_Additive
    translationMode=BMM_Additive
    alphaBoolBlend {blendInTime=0.7, blendOption=Linear, blendOutTime=0.1}
    alphaScaleBiasClamp {bInterpResult=true}
    boneToModify {boneName=dyn_skirtH_front_l_02}
   }
   n88 = "Transform (Modify) Bone - Bone: dyn_skirt_r_MSTR"!(ComponentPose=@n87, bAlphaBoolEnabled:bool=@n84, Translation="-10.939,-10.631,21.557", Rotation="47.937,140.889,148.886", Scale="1,1,1") -> (Pose) {
    alphaInputType=Bool
    rotationMode=BMM_Additive
    rotationSpace=BCS_BoneSpace
    scaleMode=BMM_Additive
    translationMode=BMM_Additive
    alphaBoolBlend {blendInTime=0.7, blendOption=Linear, blendOutTime=0.1}
    alphaScaleBiasClamp {bInterpResult=true}
    boneToModify {boneName=dyn_skirt_r_MSTR}
    rotation=<rot 47.937 140.889 148.886>
    translation=<-10.939 -10.631 21.557>
   }
   n89 = "Transform (Modify) Bone - Bone: dyn_skirt_bk_r_MSTR"!(ComponentPose=@n88, bAlphaBoolEnabled:bool=@n84, Translation="-5,-15,3", Rotation="62.255,-6.427,-28.065", Scale="1,1,1") -> (Pose) {
    alphaInputType=Bool
    rotationMode=BMM_Additive
    rotationSpace=BCS_BoneSpace
    scaleMode=BMM_Additive
    translationMode=BMM_Additive
    alphaBoolBlend {blendInTime=0.7, blendOption=Linear, blendOutTime=0.1}
    alphaScaleBiasClamp {bInterpResult=true}
    boneToModify {boneName=dyn_skirt_bk_r_MSTR}
    rotation=<rot 62.255 -6.427 -28.065>
    translation=<-5 -15 3>
   }
   n90 = "Transform (Modify) Bone - Bone: dyn_skirt_front_r_01"(ComponentPose=@n89, bAlphaBoolEnabled=@n84, Translation="-0.368,-5.242,5.402", Rotation="3.409,30.780,69.643", Scale="1,1,1") {
    alphaInputType=Bool
    rotationMode=BMM_Additive
    rotationSpace=BCS_BoneSpace
    scaleMode=BMM_Additive
    translationMode=BMM_Additive
    alphaBoolBlend {blendInTime=0.7, blendOption=Linear, blendOutTime=0.1}
    alphaScaleBiasClamp {bInterpResult=true}
    boneToModify {boneName=dyn_skirt_front_r_01}
    rotation=<rot 3.409 30.78 69.643>
    translation=<-0.368 -5.242 5.402>
   }
   n91 = "Transform (Modify) Bone - Bone: dyn_skirt_front_r_MSTR"!(ComponentPose=@n90, bAlphaBoolEnabled:bool=@n84, Translation="0,0,3", Rotation="10,0,0", Scale="1,1,1") -> (Pose) {
    alphaInputType=Bool
    rotationMode=BMM_Additive
    rotationSpace=BCS_BoneSpace
    scaleMode=BMM_Additive
    translationMode=BMM_Additive
    alphaBoolBlend {blendInTime=0.7, blendOption=Linear, blendOutTime=0.1}
    alphaScaleBiasClamp {bInterpResult=true}
    boneToModify {boneName=dyn_skirt_front_r_MSTR}
    rotation {pitch=10}
    translation {z=3}
   }
   n92 = Get bIsKneeSliding()
   n93 = "Transform (Modify) Bone - Bone: dyn_skirt_l_01"(ComponentPose=@n91, bAlphaBoolEnabled=@n92, Translation="1.285,-1.176,3.479", Rotation="50.000,-30.000,-0.000", Scale="1,1,1") {
    alphaInputType=Bool
    rotationMode=BMM_Additive
    rotationSpace=BCS_BoneSpace
    scaleMode=BMM_Additive
    translationMode=BMM_Additive
    alphaBoolBlend {blendInTime=0.4, blendOption=Linear, blendOutTime=0.1}
    alphaScaleBiasClamp {bInterpResult=true}
    boneToModify {boneName=dyn_skirt_l_01}
    rotation=<rot 50 -30 0>
    translation=<1.285 -1.176 3.479>
   }
   n94 = "Transform (Modify) Bone - Bone: dyn_skirt_bk_l_01"(ComponentPose=@n93, bAlphaBoolEnabled=@n92, Translation="-12.609,22.041,20.696", Rotation="88,0,0", Scale="1,1,1") {
    alphaInputType=Bool
    rotationMode=BMM_Additive
    rotationSpace=BCS_BoneSpace
    scaleMode=BMM_Additive
    translationMode=BMM_Additive
    alphaBoolBlend {blendInTime=0.4, blendOption=Linear, blendOutTime=0.1}
    alphaScaleBiasClamp {bInterpResult=true}
    boneToModify {boneName=dyn_skirt_bk_l_01}
    rotation {pitch=88}
    translation=<-12.609 22.041 20.696>
   }
   n95 = "Transform (Modify) Bone - Bone: dyn_skirt_r_MSTR"(ComponentPose=@n94, bAlphaBoolEnabled=@n92, Translation="6.146,-5.631,11.300", Rotation="14.478,-35.504,-3.967", Scale="1,1,1") {
    alphaInputType=Bool
    rotationMode=BMM_Additive
    rotationSpace=BCS_BoneSpace
    scaleMode=BMM_Additive
    translationMode=BMM_Additive
    alphaBoolBlend {blendInTime=0.4, blendOption=Linear, blendOutTime=0.1}
    alphaScaleBiasClamp {bInterpResult=true}
    boneToModify {boneName=dyn_skirt_r_MSTR}
    rotation=<rot 14.478 -35.504 -3.967>
    translation=<6.146 -5.631 11.3>
   }
   n96 = "Transform (Modify) Bone - Bone: dyn_skirt_bk_r_MSTR"(ComponentPose=@n95, bAlphaBoolEnabled=@n92, Translation="0.506,12.414,23.550", Rotation="77.348,-71.624,-90.447", Scale="1,1,1") {
    alphaInputType=Bool
    rotationMode=BMM_Additive
    rotationSpace=BCS_BoneSpace
    scaleMode=BMM_Additive
    translationMode=BMM_Additive
    alphaBoolBlend {blendInTime=0.4, blendOption=Linear, blendOutTime=0.1}
    alphaScaleBiasClamp {bInterpResult=true}
    boneToModify {boneName=dyn_skirt_bk_r_MSTR}
    rotation=<rot 77.348 -71.6238 -90.4468>
    translation=<0.506 12.414 23.55>
   }
   n97 = "Transform (Modify) Bone - Bone: dyn_skirt_front_r_MSTR"(ComponentPose=@n96, bAlphaBoolEnabled=@n92, Translation="0,0,0", Rotation="0,0,0", Scale="1,1,1") {
    alphaInputType=Bool
    rotationMode=BMM_Additive
    rotationSpace=BCS_BoneSpace
    scaleMode=BMM_Additive
    translationMode=BMM_Additive
    alphaBoolBlend {blendInTime=0.4, blendOption=Linear, blendOutTime=0.1}
    alphaScaleBiasClamp {bInterpResult=true}
    boneToModify {boneName=dyn_skirt_front_r_MSTR}
   }
   n98 = "Transform (Modify) Bone - Bone: guard1_MSTR"!(ComponentPose=@n97, bAlphaBoolEnabled:bool=@n92, Translation="4.155,-11.265,3.867", Rotation="-27.488,-70.723,-2.551", Scale="1,1,1") -> (Pose) {
    alphaInputType=Bool
    rotationMode=BMM_Additive
    rotationSpace=BCS_BoneSpace
    scaleMode=BMM_Additive
    translationMode=BMM_Additive
    alphaBoolBlend {blendInTime=0.4, blendOption=Linear, blendOutTime=0.1}
    alphaScaleBiasClamp {bInterpResult=true}
    boneToModify {boneName=guard1_MSTR}
    rotation=<rot -27.488 -70.723 -2.551>
    translation=<4.155 -11.265 3.867>
   }
   n99 = "Transform (Modify) Bone - Bone: guard1_MSTR"!(ComponentPose=@n98, Alpha:float=Get Is Backpedaling() -> (Is Backpedaling:double), Translation="0,0,0", Rotation="0,-20,0", Scale="1,1,1") -> (Pose) {
    rotationMode=BMM_Additive
    rotationSpace=BCS_BoneSpace
    scaleMode=BMM_Additive
    translationMode=BMM_Additive
    alphaBoolBlend {blendOption=Linear}
    alphaScaleBiasClamp {bInterpResult=true}
    boneToModify {boneName=guard1_MSTR}
    rotation {yaw=-20}
   }
   n100 = "Transform (Modify) Bone - Bone: guard1_MSTR"(ComponentPose=@n99, Alpha=Get isNotMovingNotCrouching(), Translation="0,0,3", Rotation="0,-6,0", Scale="1,1,1") {
    rotationMode=BMM_Additive
    rotationSpace=BCS_BoneSpace
    scaleMode=BMM_Additive
    translationMode=BMM_Additive
    alphaBoolBlend {blendOption=Linear}
    alphaScaleBiasClamp {bInterpResult=true, interpSpeedDecreasing=2}
    boneToModify {boneName=guard1_MSTR}
    rotation {yaw=-6}
    translation {z=3}
   }
   n101 = "Transform (Modify) Bone - Bone: guard1_MSTR"!(ComponentPose=@n100, bAlphaBoolEnabled:bool=Get bIsPlayingEmote(), Translation="1,0,3", Rotation="0,-6,0", Scale="1,1,1") -> (Pose) {
    alphaInputType=Bool
    rotationMode=BMM_Additive
    rotationSpace=BCS_BoneSpace
    scaleMode=BMM_Additive
    translationMode=BMM_Additive
    alphaBoolBlend {blendOption=Linear}
    alphaScaleBiasClamp {bInterpResult=true}
    boneToModify {boneName=guard1_MSTR}
    rotation {yaw=-6}
    translation {x=1, z=3}
   }
   n102 = "Transform (Modify) Bone - Bone: guard1_MSTR"!(ComponentPose=@n101, Alpha:float=Get Is Sky Diving(), Translation="0,0,0", Rotation="0,-20,0", Scale="1,1,1") -> (Pose) {
    rotationMode=BMM_Additive
    rotationSpace=BCS_BoneSpace
    scaleMode=BMM_Additive
    translationMode=BMM_Additive
    alphaBoolBlend {blendOption=Linear}
    alphaScaleBiasClamp {bInterpResult=true}
    boneToModify {boneName=guard1_MSTR}
    rotation {yaw=-20}
   }
   n103 = "Transform (Modify) Bone - Bone: guard1_MSTR"(ComponentPose=@n102, Alpha=Get Is Crouching(), Translation="0,0,0", Rotation="0,-20,0", Scale="1,1,1") {
    rotationMode=BMM_Additive
    rotationSpace=BCS_BoneSpace
    scaleMode=BMM_Additive
    translationMode=BMM_Additive
    alphaBoolBlend {blendOption=Linear}
    alphaScaleBiasClamp {bInterpResult=true}
    boneToModify {boneName=guard1_MSTR}
    rotation {yaw=-20}
   }
   n104 = "Transform (Modify) Bone - Bone: guard1_MSTR"(ComponentPose=@n103, Alpha=Get Is Sprinting() -> (Is Sprinting:double), Translation="0,0,0", Rotation="0,-20,0", Scale="1,1,1") {
    rotationMode=BMM_Additive
    rotationSpace=BCS_BoneSpace
    scaleMode=BMM_Additive
    translationMode=BMM_Additive
    alphaBoolBlend {blendOption=Linear}
    alphaScaleBiasClamp {bInterpResult=true}
    boneToModify {boneName=guard1_MSTR}
    rotation {yaw=-20}
   }
   n105 = "Transform (Modify) Bone - Bone: guard1_MSTR"(ComponentPose=@n104, Alpha=Get isJumpingNoSkydive() -> (isJumpingNoSkydive:double), Translation="0,0,0", Rotation="0,15,0", Scale="1,1,1") {
    rotationMode=BMM_Additive
    rotationSpace=BCS_BoneSpace
    scaleMode=BMM_Additive
    translationMode=BMM_Additive
    alphaBoolBlend {blendOption=Linear}
    alphaScaleBiasClamp {bInterpResult=true, interpSpeedDecreasing=5, interpSpeedIncreasing=20}
    boneToModify {boneName=guard1_MSTR}
    rotation {yaw=15}
   }
   n106 = "Transform (Modify) Bone - Bone: guard1_MSTR"(ComponentPose=@n105, Alpha=Get isFallingNoSkydive() -> (isFallingNoSkydive:double), Translation="0,0,0", Rotation="0,-10,0", Scale="1,1,1") {
    rotationMode=BMM_Additive
    rotationSpace=BCS_BoneSpace
    scaleMode=BMM_Additive
    translationMode=BMM_Additive
    alphaBoolBlend {blendOption=Linear}
    alphaScaleBiasClamp {bInterpResult=true, interpSpeedIncreasing=0.5}
    boneToModify {boneName=guard1_MSTR}
    rotation {yaw=-10}
   }
   n107 = "Transform (Modify) Bone - Bone: dyn_skirt_r_02"(ComponentPose=@n106, Alpha=@n57, Translation="0,0,0", Rotation="5,0,0", Scale="1,1,1") {
    rotationMode=BMM_Additive
    rotationSpace=BCS_BoneSpace
    scaleMode=BMM_Additive
    translationMode=BMM_Additive
    alphaBoolBlend {blendOption=Linear}
    alphaScaleBiasClamp {bInterpResult=true, interpSpeedDecreasing=20, interpSpeedIncreasing=20}
    boneToModify {boneName=dyn_skirt_r_02}
    rotation {pitch=5}
   }
   n108 = "Transform (Modify) Bone - Bone: dyn_skirt_front_r_02"(ComponentPose=@n107, Alpha=@n57, Translation="0,0,0", Rotation="0,-15,0", Scale="1,1,1") {
    rotationMode=BMM_Additive
    rotationSpace=BCS_BoneSpace
    scaleMode=BMM_Additive
    translationMode=BMM_Additive
    alphaBoolBlend {blendOption=Linear}
    alphaScaleBiasClamp {bInterpResult=true, interpSpeedDecreasing=20, interpSpeedIncreasing=20}
    boneToModify {boneName=dyn_skirt_front_r_02}
    rotation {yaw=-15}
   }
   n109 = "Transform (Modify) Bone - Bone: dyn_skirt_bk_r_MSTR"!(ComponentPose=@n108, Alpha:float=@n57, Translation="0,0,0", Rotation="15,20,0", Scale="1,1,1") -> (Pose) {
    rotationMode=BMM_Additive
    rotationSpace=BCS_BoneSpace
    scaleMode=BMM_Additive
    translationMode=BMM_Additive
    alphaBoolBlend {blendOption=Linear}
    alphaScaleBiasClamp {bInterpResult=true, interpSpeedIncreasing=1}
    boneToModify {boneName=dyn_skirt_bk_r_MSTR}
    rotation {pitch=15, yaw=20}
   }
   n110 = "Transform (Modify) Bone - Bone: dyn_skirt_bk_l_MSTR"(ComponentPose=@n109, Alpha=@n57, Translation="0,0,0", Rotation="15,10,0", Scale="1,1,1") {
    rotationMode=BMM_Additive
    rotationSpace=BCS_BoneSpace
    scaleMode=BMM_Additive
    translationMode=BMM_Additive
    alphaBoolBlend {blendOption=Linear}
    alphaScaleBiasClamp {bInterpResult=true}
    boneToModify {boneName=dyn_skirt_bk_l_MSTR}
    rotation {pitch=15, yaw=10}
   }
   n111 = Get LeftMovementClean() -> (LeftMovementClean:double)
   n112 = "Transform (Modify) Bone - Bone: dyn_skirt_l_02"(ComponentPose=@n110, Alpha=@n111, Translation="0,0,0", Rotation="5,0,0", Scale="1,1,1") {
    rotationMode=BMM_Additive
    rotationSpace=BCS_BoneSpace
    scaleMode=BMM_Additive
    translationMode=BMM_Additive
    alphaBoolBlend {blendOption=Linear}
    alphaScaleBiasClamp {bInterpResult=true}
    boneToModify {boneName=dyn_skirt_l_02}
    rotation {pitch=5}
   }
   n113 = "Transform (Modify) Bone - Bone: dyn_skirt_bk_l_02"(ComponentPose=@n112, Alpha=@n111, Translation="0,0,0", Rotation="15,5,0", Scale="1,1,1") {
    rotationMode=BMM_Additive
    rotationSpace=BCS_BoneSpace
    scaleMode=BMM_Additive
    translationMode=BMM_Additive
    alphaBoolBlend {blendOption=Linear}
    alphaScaleBiasClamp {bInterpResult=true, interpSpeedIncreasing=1}
    boneToModify {boneName=dyn_skirt_bk_l_02}
    rotation {pitch=15, yaw=5}
   }
   n114 = "Transform (Modify) Bone - Bone: dyn_skirtH_front_l_02"!(ComponentPose=@n113, Alpha:float=@n111, Translation="0,0,0", Rotation="5,-5,0", Scale="1,1,1") -> (Pose) {
    rotationMode=BMM_Additive
    rotationSpace=BCS_BoneSpace
    scaleMode=BMM_Additive
    translationMode=BMM_Additive
    alphaBoolBlend {blendOption=Linear}
    alphaScaleBiasClamp {bInterpResult=true}
    boneToModify {boneName=dyn_skirtH_front_l_02}
    rotation {pitch=5, yaw=-5}
   }
   n115 = "Transform (Modify) Bone - Bone: dyn_skirt_bk_r_MSTR"(ComponentPose=@n114, Alpha=@n111, Translation="0,0,0", Rotation="0,20,0", Scale="1,1,1") {
    rotationMode=BMM_Additive
    rotationSpace=BCS_BoneSpace
    scaleMode=BMM_Additive
    translationMode=BMM_Additive
    alphaBoolBlend {blendOption=Linear}
    alphaScaleBiasClamp {bInterpResult=true, interpSpeedDecreasing=20, interpSpeedIncreasing=5}
    boneToModify {boneName=dyn_skirt_bk_r_MSTR}
    rotation {yaw=20}
   }
   n116 = "Transform (Modify) Bone - Bone: dyn_skirt_bk_r_04"(ComponentPose=@n115, Alpha=@n111, Translation="0,0,0", Rotation="10,0,0", Scale="1,1,1") {
    rotationMode=BMM_Additive
    rotationSpace=BCS_BoneSpace
    scaleMode=BMM_Additive
    translationMode=BMM_Additive
    alphaBoolBlend {blendOption=Linear}
    alphaScaleBiasClamp {bInterpResult=true, interpSpeedDecreasing=20}
    boneToModify {boneName=dyn_skirt_bk_r_04}
    rotation {pitch=10}
   }
   n117 = "Transform (Modify) Bone - Bone: dyn_skirt_bk_l_MSTR"(ComponentPose=@n116, Alpha=@n111, Translation="0,0,0", Rotation="5,10,0", Scale="1,1,1") {
    rotationMode=BMM_Additive
    rotationSpace=BCS_BoneSpace
    scaleMode=BMM_Additive
    translationMode=BMM_Additive
    alphaBoolBlend {blendOption=Linear}
    alphaScaleBiasClamp {bInterpResult=true, interpSpeedIncreasing=1}
    boneToModify {boneName=dyn_skirt_bk_l_MSTR}
    rotation {pitch=5, yaw=10}
   }
   n118 = Get isCrouchingRightMovementClean() -> (isCrouchingRightMovementClean:double)
   n119 = "Transform (Modify) Bone - Bone: dyn_skirt_r_02"(ComponentPose=@n117, Alpha=@n118, Translation="0,0,0", Rotation="5,0,0", Scale="1,1,1") {
    rotationMode=BMM_Additive
    rotationSpace=BCS_BoneSpace
    scaleMode=BMM_Additive
    translationMode=BMM_Additive
    alphaBoolBlend {blendOption=Linear}
    alphaScaleBiasClamp {bInterpResult=true, interpSpeedDecreasing=20, interpSpeedIncreasing=20}
    boneToModify {boneName=dyn_skirt_r_02}
    rotation {pitch=5}
   }
   n120 = "Transform (Modify) Bone - Bone: dyn_skirt_front_r_02"(ComponentPose=@n119, Alpha=@n118, Translation="0,0,0", Rotation="0,-5,0", Scale="1,1,1") {
    rotationMode=BMM_Additive
    rotationSpace=BCS_BoneSpace
    scaleMode=BMM_Additive
    translationMode=BMM_Additive
    alphaBoolBlend {blendOption=Linear}
    alphaScaleBiasClamp {bInterpResult=true, interpSpeedDecreasing=20, interpSpeedIncreasing=20}
    boneToModify {boneName=dyn_skirt_front_r_02}
    rotation {yaw=-5}
   }
   n121 = "Transform (Modify) Bone - Bone: dyn_skirt_front_r_04"(ComponentPose=@n120, Alpha=@n118, Translation="0,0,0", Rotation="0,-5,0", Scale="1,1,1") {
    rotationMode=BMM_Additive
    rotationSpace=BCS_BoneSpace
    scaleMode=BMM_Additive
    translationMode=BMM_Additive
    alphaBoolBlend {blendOption=Linear}
    alphaScaleBiasClamp {bInterpResult=true, interpSpeedDecreasing=20, interpSpeedIncreasing=20}
    boneToModify {boneName=dyn_skirt_front_r_04}
    rotation {yaw=-5}
   }
   n122 = "Transform (Modify) Bone - Bone: dyn_skirt_front_r_05"(ComponentPose=@n121, Alpha=@n118, Translation="0,0,0", Rotation="0,-5,0", Scale="1,1,1") {
    rotationMode=BMM_Additive
    rotationSpace=BCS_BoneSpace
    scaleMode=BMM_Additive
    translationMode=BMM_Additive
    alphaBoolBlend {blendOption=Linear}
    alphaScaleBiasClamp {bInterpResult=true, interpSpeedDecreasing=20, interpSpeedIncreasing=20}
    boneToModify {boneName=dyn_skirt_front_r_05}
    rotation {yaw=-5}
   }
   n123 = "Transform (Modify) Bone - Bone: dyn_skirt_bk_r_MSTR"(ComponentPose=@n122, Alpha=@n118, Translation="0,0,0", Rotation="0,20,0", Scale="1,1,1") {
    rotationMode=BMM_Additive
    rotationSpace=BCS_BoneSpace
    scaleMode=BMM_Additive
    translationMode=BMM_Additive
    alphaBoolBlend {blendOption=Linear}
    alphaScaleBiasClamp {bInterpResult=true, interpSpeedIncreasing=2}
    boneToModify {boneName=dyn_skirt_bk_r_MSTR}
    rotation {yaw=20}
   }
   n124 = "Transform (Modify) Bone - Bone: dyn_skirt_r_MSTR"!(ComponentPose=@n123, Alpha:float=@n118, Translation="0,0,0", Rotation="0,10,0", Scale="1,1,1") -> (Pose) {
    rotationMode=BMM_Additive
    rotationSpace=BCS_BoneSpace
    scaleMode=BMM_Additive
    translationMode=BMM_Additive
    alphaBoolBlend {blendOption=Linear}
    alphaScaleBiasClamp {bInterpResult=true}
    boneToModify {boneName=dyn_skirt_r_MSTR}
    rotation {yaw=10}
   }
   n125 = "Transform (Modify) Bone - Bone: dyn_skirt_bk_l_MSTR"(ComponentPose=@n124, Alpha=@n118, Translation="0,0,0", Rotation="0,30,0", Scale="1,1,1") {
    rotationMode=BMM_Additive
    rotationSpace=BCS_BoneSpace
    scaleMode=BMM_Additive
    translationMode=BMM_Additive
    alphaBoolBlend {blendOption=Linear}
    alphaScaleBiasClamp {bInterpResult=true, interpSpeedIncreasing=2}
    boneToModify {boneName=dyn_skirt_bk_l_MSTR}
    rotation {yaw=30}
   }
   n126 = Get isCrouchingRLeftMovementClean() -> (isCrouchingRLeftMovementClean:double)
   n127 = "Transform (Modify) Bone - Bone: dyn_skirt_l_02"(ComponentPose=@n125, Alpha=@n126, Translation="0,0,0", Rotation="10,0,0", Scale="1,1,1") {
    rotationMode=BMM_Additive
    rotationSpace=BCS_BoneSpace
    scaleMode=BMM_Additive
    translationMode=BMM_Additive
    alphaBoolBlend {blendOption=Linear}
    alphaScaleBiasClamp {bInterpResult=true}
    boneToModify {boneName=dyn_skirt_l_02}
    rotation {pitch=10}
   }
   n128 = "Transform (Modify) Bone - Bone: dyn_skirt_bk_l_02"(ComponentPose=@n127, Alpha=@n126, Translation="0,0,0", Rotation="5,5,0", Scale="1,1,1") {
    rotationMode=BMM_Additive
    rotationSpace=BCS_BoneSpace
    scaleMode=BMM_Additive
    translationMode=BMM_Additive
    alphaBoolBlend {blendOption=Linear}
    alphaScaleBiasClamp {bInterpResult=true, interpSpeedIncreasing=2}
    boneToModify {boneName=dyn_skirt_bk_l_02}
    rotation {pitch=5, yaw=5}
   }
   n129 = "Transform (Modify) Bone - Bone: dyn_skirtH_front_l_02"(ComponentPose=@n128, Alpha=@n126, Translation="0,0,0", Rotation="5,-5,0", Scale="1,1,1") {
    rotationMode=BMM_Additive
    rotationSpace=BCS_BoneSpace
    scaleMode=BMM_Additive
    translationMode=BMM_Additive
    alphaBoolBlend {blendOption=Linear}
    alphaScaleBiasClamp {bInterpResult=true}
    boneToModify {boneName=dyn_skirtH_front_l_02}
    rotation {pitch=5, yaw=-5}
   }
   n130 = "Transform (Modify) Bone - Bone: dyn_skirt_bk_r_MSTR"(ComponentPose=@n129, Alpha=@n126, Translation="0,0,0", Rotation="0,15,0", Scale="1,1,1") {
    rotationMode=BMM_Additive
    rotationSpace=BCS_BoneSpace
    scaleMode=BMM_Additive
    translationMode=BMM_Additive
    alphaBoolBlend {blendOption=Linear}
    alphaScaleBiasClamp {bInterpResult=true, interpSpeedDecreasing=20, interpSpeedIncreasing=2}
    boneToModify {boneName=dyn_skirt_bk_r_MSTR}
    rotation {yaw=15}
   }
   n131 = "Transform (Modify) Bone - Bone: dyn_skirt_bk_l_MSTR"(ComponentPose=@n130, Alpha=@n126, Translation="0,0,0", Rotation="0,10,0", Scale="1,1,1") {
    rotationMode=BMM_Additive
    rotationSpace=BCS_BoneSpace
    scaleMode=BMM_Additive
    translationMode=BMM_Additive
    alphaBoolBlend {blendOption=Linear}
    alphaScaleBiasClamp {bInterpResult=true, interpSpeedIncreasing=2}
    boneToModify {boneName=dyn_skirt_bk_l_MSTR}
    rotation {yaw=10}
   }
   n132 = Get bIsTacticalSprinting()
   n133 = "Transform (Modify) Bone - Bone: dyn_skirt_bk_l_03"!(ComponentPose=@n131, bAlphaBoolEnabled:bool=@n132, Translation="0,0,0", Rotation="20,30,0", Scale="1,1,1") -> (Pose) {
    alphaInputType=Bool
    rotationMode=BMM_Additive
    rotationSpace=BCS_BoneSpace
    scaleMode=BMM_Additive
    scaleSpace=BCS_BoneSpace
    translationMode=BMM_Additive
    translationSpace=BCS_BoneSpace
    alphaBoolBlend {blendInTime=0.1, blendOption=Linear, blendOutTime=0.1}
    boneToModify {boneName=dyn_skirt_bk_l_03}
    rotation {pitch=20, yaw=30}
   }
   n134 = "Transform (Modify) Bone - Bone: dyn_skirt_bk_r_MSTR"!(ComponentPose=@n133, bAlphaBoolEnabled:bool=@n132, Translation="0,0,0", Rotation="20,0,0", Scale="1,1,1") -> (Pose) {
    alphaInputType=Bool
    rotationMode=BMM_Additive
    rotationSpace=BCS_BoneSpace
    scaleMode=BMM_Additive
    scaleSpace=BCS_BoneSpace
    translationMode=BMM_Additive
    translationSpace=BCS_BoneSpace
    alphaBoolBlend {blendInTime=0.1, blendOption=Linear, blendOutTime=0.1}
    boneToModify {boneName=dyn_skirt_bk_r_MSTR}
    rotation {pitch=20}
   }
   n135 = "Transform (Modify) Bone - Bone: guard1_MSTR"!(ComponentPose=@n134, bAlphaBoolEnabled:bool=@n132, Translation="0,0,0", Rotation="0,-40,0", Scale="1,1,1") -> (Pose) {
    alphaInputType=Bool
    rotationMode=BMM_Additive
    rotationSpace=BCS_BoneSpace
    scaleMode=BMM_Additive
    translationMode=BMM_Additive
    alphaBoolBlend {blendInTime=0.1, blendOption=Linear, blendOutTime=0.1}
    alphaScaleBiasClamp {bInterpResult=true}
    boneToModify {boneName=guard1_MSTR}
    rotation {yaw=-40}
   }
   n136 = Get Backward() -> (Backward:double)
   n137 = "Transform (Modify) Bone - Bone: dyn_skirt_bk_r_MSTR"!(ComponentPose=@n135, Alpha:float=@n136, Translation="0,0,0", Rotation="8,0,0", Scale="1,1,1") -> (Pose) {
    rotationMode=BMM_Additive
    rotationSpace=BCS_BoneSpace
    scaleMode=BMM_Additive
    scaleSpace=BCS_BoneSpace
    translationMode=BMM_Additive
    translationSpace=BCS_BoneSpace
    alphaBoolBlend {blendOption=Linear}
    alphaScaleBiasClamp {bInterpResult=true}
    boneToModify {boneName=dyn_skirt_bk_r_MSTR}
    rotation {pitch=8}
   }
   n138 = "Transform (Modify) Bone - Bone: dyn_skirt_bk_l_MSTR"(ComponentPose=@n137, Alpha=@n136, Translation="0,0,0", Rotation="8,0,0", Scale="1,1,1") {
    rotationMode=BMM_Additive
    rotationSpace=BCS_BoneSpace
    scaleMode=BMM_Additive
    scaleSpace=BCS_BoneSpace
    translationMode=BMM_Additive
    translationSpace=BCS_BoneSpace
    alphaBoolBlend {blendOption=Linear}
    alphaScaleBiasClamp {bInterpResult=true}
    boneToModify {boneName=dyn_skirt_bk_l_MSTR}
    rotation {pitch=8}
   }
   n139 = "Transform (Modify) Bone - Bone: dyn_skirt_bk_l_04"(ComponentPose=@n138, Alpha=@n136, Translation="0,0,0", Rotation="20,0,0", Scale="1,1,1") {
    rotationMode=BMM_Additive
    rotationSpace=BCS_BoneSpace
    scaleMode=BMM_Additive
    scaleSpace=BCS_BoneSpace
    translationMode=BMM_Additive
    translationSpace=BCS_BoneSpace
    alphaBoolBlend {blendOption=Linear}
    alphaScaleBiasClamp {bInterpResult=true}
    boneToModify {boneName=dyn_skirt_bk_l_04}
    rotation {pitch=20}
   }
   n140 = "Transform (Modify) Bone - Bone: dyn_skirt_bk_r_MSTR"(ComponentPose=@n139, Alpha=Get isJumpingNoSkydive(), Translation="0,0,0", Rotation="40,15,0", Scale="1,1,1") {
    rotationMode=BMM_Additive
    rotationSpace=BCS_BoneSpace
    scaleMode=BMM_Additive
    translationMode=BMM_Additive
    alphaBoolBlend {blendOption=Linear}
    alphaScaleBiasClamp {bInterpResult=true, interpSpeedDecreasing=1}
    boneToModify {boneName=dyn_skirt_bk_r_MSTR}
    rotation {pitch=40, yaw=15}
   }
   n141 = Get bIsParachuteOpened()
   n142 = "Transform (Modify) Bone - Bone: guard1_MSTR"(ComponentPose=@n140, bAlphaBoolEnabled=@n141, Translation="-3,0,5", Rotation="0,8,0", Scale="1,1,1") {
    alphaInputType=Bool
    rotationMode=BMM_Additive
    rotationSpace=BCS_BoneSpace
    scaleMode=BMM_Additive
    translationMode=BMM_Additive
    alphaBoolBlend {blendOption=Linear}
    alphaScaleBiasClamp {bInterpResult=true, interpSpeedDecreasing=1}
    boneToModify {boneName=guard1_MSTR}
    rotation {yaw=8}
    translation {x=-3, z=5}
   }
   n143 = "Transform (Modify) Bone - Bone: dyn_skirt_bk_l_MSTR"!(ComponentPose=@n142, bAlphaBoolEnabled:bool=@n141, Translation="5,0,0", Rotation="0,0,0", Scale="1,1,1") -> (Pose) {
    alphaInputType=Bool
    rotationMode=BMM_Additive
    rotationSpace=BCS_BoneSpace
    scaleMode=BMM_Additive
    translationMode=BMM_Additive
    alphaBoolBlend {blendOption=Linear}
    alphaScaleBiasClamp {bInterpResult=true, interpSpeedDecreasing=1}
    boneToModify {boneName=dyn_skirt_bk_l_MSTR}
    translation {x=5}
   }
   n144 = "Transform (Modify) Bone - Bone: dyn_skirt_bk_l_02"!(ComponentPose=@n143, bAlphaBoolEnabled:bool=@n141, Translation="-10,-3,0", Rotation="0,0,0", Scale="1,1,1") -> (Pose) {
    alphaInputType=Bool
    rotationMode=BMM_Additive
    rotationSpace=BCS_BoneSpace
    scaleMode=BMM_Additive
    translationMode=BMM_Additive
    alphaBoolBlend {blendOption=Linear}
    alphaScaleBiasClamp {bInterpResult=true, interpSpeedDecreasing=1}
    boneToModify {boneName=dyn_skirt_bk_l_02}
    translation=<-10 -3>
   }
   n145 = Get inFrontEndNotEmoting()
   n146 = "Transform (Modify) Bone - Bone: dyn_skirt_bk_l_03"(ComponentPose=@n144, bAlphaBoolEnabled=@n145, Translation="0,0,0", Rotation="-10,10,0", Scale="1,1,1") {
    alphaInputType=Bool
    rotationMode=BMM_Additive
    rotationSpace=BCS_BoneSpace
    scaleMode=BMM_Additive
    translationMode=BMM_Additive
    alphaBoolBlend {blendOption=Linear}
    alphaScaleBiasClamp {bInterpResult=true, interpSpeedDecreasing=1}
    boneToModify {boneName=dyn_skirt_bk_l_03}
    rotation {pitch=-10, yaw=10}
   }
   n147 = "Transform (Modify) Bone - Bone: dyn_skirt_l_MSTR"!(ComponentPose=@n146, bAlphaBoolEnabled:bool=@n145, Translation="0,0,0", Rotation="-5,0,0", Scale="1,1,1") -> (Pose) {
    alphaInputType=Bool
    rotationMode=BMM_Additive
    rotationSpace=BCS_BoneSpace
    scaleMode=BMM_Additive
    translationMode=BMM_Additive
    alphaBoolBlend {blendOption=Linear}
    alphaScaleBiasClamp {bInterpResult=true, interpSpeedDecreasing=1}
    boneToModify {boneName=dyn_skirt_l_MSTR}
    rotation {pitch=-5}
   }
   n148 = "Transform (Modify) Bone - Bone: dyn_skirtH_front_l_MSTR"(ComponentPose=@n147, bAlphaBoolEnabled=@n145, Translation="0,0,0", Rotation="0,10,0", Scale="1,1,1") {
    alphaInputType=Bool
    rotationMode=BMM_Additive
    rotationSpace=BCS_BoneSpace
    scaleMode=BMM_Additive
    translationMode=BMM_Additive
    alphaBoolBlend {blendOption=Linear}
    alphaScaleBiasClamp {bInterpResult=true, interpSpeedDecreasing=1}
    boneToModify {boneName=dyn_skirtH_front_l_MSTR}
    rotation {yaw=10}
   }
   n149 = "Transform (Modify) Bone - Bone: dyn_skirtH_front_l_03"!(ComponentPose=@n148, bAlphaBoolEnabled:bool=@n145, Translation="0,0,0", Rotation="-5,0,0", Scale="1,1,1") -> (Pose) {
    alphaInputType=Bool
    rotationMode=BMM_Additive
    rotationSpace=BCS_BoneSpace
    scaleMode=BMM_Additive
    translationMode=BMM_Additive
    alphaBoolBlend {blendOption=Linear}
    alphaScaleBiasClamp {bInterpResult=true, interpSpeedDecreasing=1}
    boneToModify {boneName=dyn_skirtH_front_l_03}
    rotation {pitch=-5}
   }
   n150 = "Transform (Modify) Bone - Bone: dyn_skirtH_front_l_04"!(ComponentPose=@n149, bAlphaBoolEnabled:bool=@n145, Translation="0,0,0", Rotation="0,-5,0", Scale="1,1,1") -> (Pose) {
    alphaInputType=Bool
    rotationMode=BMM_Additive
    rotationSpace=BCS_BoneSpace
    scaleMode=BMM_Additive
    translationMode=BMM_Additive
    alphaBoolBlend {blendOption=Linear}
    alphaScaleBiasClamp {bInterpResult=true, interpSpeedDecreasing=1}
    boneToModify {boneName=dyn_skirtH_front_l_04}
    rotation {yaw=-5}
   }
   n151 = "Transform (Modify) Bone - Bone: dyn_skirt_bk_r_03"!(ComponentPose=@n150, bAlphaBoolEnabled:bool=@n145, Translation="0,0,0", Rotation="-5,10,0", Scale="1,1,1") -> (Pose) {
    alphaInputType=Bool
    rotationMode=BMM_Additive
    rotationSpace=BCS_BoneSpace
    scaleMode=BMM_Additive
    translationMode=BMM_Additive
    alphaBoolBlend {blendOption=Linear}
    alphaScaleBiasClamp {bInterpResult=true, interpSpeedDecreasing=1}
    boneToModify {boneName=dyn_skirt_bk_r_03}
    rotation {pitch=-5, yaw=10}
   }
   Output Pose(Result=Component To Local(ComponentPose=@n151)) {layerGroup=None}
   Get isNotMoving() -> (isNotMoving:double)
   Get bIsAccelerating2D()
   Get Is Moving 2D()
   Get Is Jumping()
  ]}
 EventGraph {nodes [
   Event BlueprintUpdateAnimation()
   > n1 = Sequence!(execute) -> (then_0, then_1, then_2:Exec, then_3:Exec, then_4:Exec)
   > n2 = Gravity Override State Function DEPRECATED(CharacterPartAnimInstance=Self-Reference(), GameTimeInSeconds=GetGameTimeInSeconds(), GravityOverrideParams=Get Gravity Override Params())
   > Set GravityOverride(GravityOverride=@n2)
   n3 = BreakTransform(InTransform=GetDeltaTransformFromRefPose(self=GetOwnerSkeletalMeshComponent(), BoneName=thigh_l, BaseName=pelvis))
   Set thigh_up_l(execute=@n1.then_1, thigh_up_l:double=MapRangeClamped(Value=BreakRotator(InRot=@n3.Rotation).Yaw, InRangeA=-30, InRangeB=-70)) -> (Output_Get:double)
   n4 = BreakTransform(InTransform=GetDeltaTransformFromRefPose(self=GetOwnerSkeletalMeshComponent(), BoneName=thigh_r, BaseName=pelvis))
   > Set thigh_up_r(thigh_up_r:double=MapRangeClamped(Value=BreakRotator(InRot=@n4.Rotation).Yaw, InRangeA=0, InRangeB=-40)) -> (Output_Get:double)
   n5 = BreakTransform(InTransform=GetDeltaTransformFromRefPose(self=GetOwnerSkeletalMeshComponent(), BoneName=thigh_l, BaseName=pelvis))
   > Set ThighSideLf(ThighSideLf:double=MapRangeClamped(Value=BreakRotator(InRot=@n5.Rotation).Pitch, InRangeA=0, InRangeB=40)) -> (Output_Get:double)
   n6 = BreakTransform(InTransform=GetDeltaTransformFromRefPose(self=GetOwnerSkeletalMeshComponent(), BoneName=thigh_r, BaseName=pelvis))
   > Set ThighSideRt(ThighSideRt:double=MapRangeClamped(Value=BreakRotator(InRot=@n6.Rotation).Pitch, InRangeA=0, InRangeB=40)) -> (Output_Get:double)
   n7 = BreakTransform(InTransform=GetDeltaTransformFromRefPose(self=GetOwnerSkeletalMeshComponent(), BoneName=calf_l, BaseName=thigh_l))
   > Set KneeLf(KneeLf:double=MapRangeClamped(Value=BreakRotator(InRot=@n7.Rotation).Yaw, InRangeA=0, InRangeB=100)) -> (Output_Get:double)
   Set KneeStraight(execute=@n1.then_2, KneeStraight:bool="Not Equal (Float)"(A=Get KneeLf() -> (KneeLf:double), B=1)) -> (Output_Get:bool)
   n8 = Get bIsPlayingEmote()
   Set inFrontEndNotEmoting(execute=@n1.then_4, inFrontEndNotEmoting:bool=AND Boolean(A=Get bIsInFrontEnd(), B=NOT Boolean(A=@n8))) -> (Output_Get:bool)
   n9 = Get bIsInFrontEnd()
   > Set Is NotIn Front End(Is NotIn Front End:double="To Float (Boolean)"(InBool=NOT Boolean(A=@n9))) -> (Output_Get:double)
   > Set isNotSkydivingNotInFE(isNotSkydivingNotInFE:double=float * float(A=Get isNotSkydiving() -> (isNotSkydiving:double), B=Get Is NotIn Front End() -> (Is NotIn Front End:double)) {is=title K2Node_PromotableOperator}) -> (Output_Get:double)
   > Set isInFrontEndIsEmoting(isInFrontEndIsEmoting:bool=AND Boolean(A=Get bIsPlayingEmote(), B=Get bIsInFrontEnd())) -> (Output_Get:bool)
   n10 = Get bIsCrouching()
   Set Is Crouching(execute=@n1.then_3, Is Crouching:double="To Float (Boolean)"(InBool=@n10)) -> (Output_Get:double)
   > Set isNotCrouching(isNotCrouching:double="To Float (Boolean)"(InBool=NOT Boolean(A=@n10))) -> (Output_Get:double)
   n11 = Get bIsSkyDiving()
   n12 = Get grinding() -> (grinding:bool)
   n13 = Get bIsDBNO()
   n14 = Get bIsSliding()
   > Set Is Sky Diving(Is Sky Diving:double="To Float (Boolean)"(InBool=OR Boolean!(A=@n11, B=@n12, C:bool=@n13, D:bool=@n14) -> (ReturnValue))) -> (Output_Get:double)
   n15 = OR Boolean(A=Get bIsSkyDiving(), B=Get grinding(), C=Get bIsDBNO(), D=Get bIsSliding())
   > Set isNotSkydiving(isNotSkydiving:double="To Float (Boolean)"(InBool=NOT Boolean(A=@n15))) -> (Output_Get:double)
   n16 = Get bIsBackpedaling() -> (bIsBackpedaling:bool)
   > Set Is Backpedaling(Is Backpedaling:double="To Float (Boolean)"(InBool=@n16)) -> (Output_Get:double)
   > Set isNotBackpeddling(isNotBackpeddling:double="To Float (Boolean)"(InBool=NOT Boolean(A=@n16))) -> (Output_Get:double)
   n17 = Get bIsSprinting()
   > Set Is Sprinting(Is Sprinting:double="To Float (Boolean)"(InBool=@n17)) -> (Output_Get:double)
   > Set isNotSprinting(isNotSprinting:double="To Float (Boolean)"(InBool=NOT Boolean(A=@n17))) -> (Output_Get:double)
   > Set Is Sliding(Is Sliding:double="To Float (Boolean)"(InBool=Get bIsSliding())) -> (Output_Get:double)
   n18 = Get bIsSliding()
   > Set isNotSliding(isNotSliding:double="To Float (Boolean)"(InBool=NOT Boolean(A=@n18))) -> (Output_Get:double)
   > Set isNotSkydivingNotSliding(isNotSkydivingNotSliding:double=float * float(A=Get isNotSliding() -> (isNotSliding:double), B=Get isNotSkydiving())) -> (Output_Get:double)
   n19 = Get bIsFalling()
   > Set Is Falling(Is Falling:double="To Float (Boolean)"(InBool=@n19)) -> (Output_Get:double)
   n20 = Get bIsJumping()
   > Set Is Jumping(Is Jumping:double="To Float (Boolean)"(InBool=OR Boolean!(A=@n19, B=@n20) -> (ReturnValue))) -> (Output_Get:double)
   n21 = Get isNotSkydiving()
   > Set isFallingNoSkydive(isFallingNoSkydive:double=float * float(A=Get Is Falling() -> (Is Falling:double), B=@n21)) -> (Output_Get:double)
   > Set isJumpingNoSkydive(isJumpingNoSkydive:double=float * float(A=@n21, B=Get Is Jumping())) -> (Output_Get:double)
   > Set Is Moving 2D(Is Moving 2D:double="To Float (Boolean)"(InBool=Get bIsMoving2D())) -> (Output_Get:double)
   n22 = Get bIsMoving2D()
   > Set isNotMoving(isNotMoving:double="To Float (Boolean)"(InBool=NOT Boolean(A=@n22))) -> (Output_Get:double)
   n23 = Get Is Crouching()
   > Set isCrouchingNotMoving(isCrouchingNotMoving:double=float * float(A=Get isNotMoving(), B=@n23) {is=title K2Node_CommutativeAssociativeBinaryOperator}) -> (Output_Get:double)
   > Set isCrouchMoving(isCrouchMoving:double=float * float(A=Get Is Moving 2D(), B=@n23)) -> (Output_Get:double)
   > Set Is Skydive Floating(Is Skydive Floating:double="To Float (Boolean)"(InBool=Get bIsSkydiveFloating())) -> (Output_Get:double)
   > Set isNotMovingNotCrouching(isNotMovingNotCrouching:double=float * float(A=Get isNotMoving(), B=Get isNotCrouching() -> (isNotCrouching:double))) -> (Output_Get:double)
   > n24 = PlayerMovement(AnimInstance=Self-Reference(), XDotProductCoefficient=1, YDotProductCoefficient=1, ZDotProductCoefficient=1)
   > Set Backward(Backward:double=MapRangeClamped(Value=@n24.YMovement, InRangeA=0, InRangeB=-1)) -> (Output_Get:double)
   n25 = MapRangeClamped(Value=@n24, InRangeA=0, InRangeB=1)
   n26 = float * float(A="Absolute (Float)"(A=@n24.YMovement), B=0.2)
   > Set RightMovement(RightMovement:double=MapRangeClamped(Value=float + float(A=@n26, B=@n25) {is=title K2Node_CommutativeAssociativeBinaryOperator}, InRangeA=0, InRangeB=1.2, OutRangeB=0.5)) -> (Output_Get:double)
   n27 = MapRangeClamped(Value=@n24, InRangeA=0, InRangeB=-1, OutRangeB=0.3)
   n28 = float * float(A="Absolute (Float)"(A=@n24.YMovement), B=0.2)
   > Set LeftMovement(LeftMovement:double=MapRangeClamped(Value=float + float(A=@n28, B=@n27), InRangeA=0, InRangeB=1.2, OutRangeB=0.5)) -> (Output_Get:double)
   n29 = Get isNotSprinting() -> (isNotSprinting:double)
   n30 = Get isNotCrouching()
   n31 = Get isNotSkydiving()
   n32 = Get isNotBackpeddling() -> (isNotBackpeddling:double)
   > Set RightMovementClean(RightMovementClean:double=float * float!(A=Get RightMovement() -> (RightMovement:double), B=@n31, C:double=@n30, D:double=@n32, E:double=@n29, F:double=2.5) -> (ReturnValue) {is=title K2Node_PromotableOperator}) -> (Output_Get:double)
   > Set LeftMovementClean(LeftMovementClean:double=float * float(A=Get LeftMovement() -> (LeftMovement:double), B=@n31, C=@n30, D=@n32, E=@n29, F=8.5)) -> (Output_Get:double)
   n33 = Get isNotSkydiving()
   n34 = Get isNotBackpeddling()
   n35 = Get isNotSprinting()
   n36 = Get Is Crouching()
   > Set isCrouchingRightMovementClean(isCrouchingRightMovementClean:double=float * float(A=Get RightMovement(), B=@n33, C=@n36, D=@n34, E=@n35, F=2.5)) -> (Output_Get:double)
   > Set isCrouchingRLeftMovementClean(isCrouchingRLeftMovementClean:double=float * float(A=Get LeftMovement(), B=@n33, C=@n36, D=@n34, E=@n35, F=8.5)) -> (Output_Get:double)
   Event OnEnterGrindRail() {is=title K2Node_Event}
   > Set grinding(grinding:bool=true) -> (Output_Get:bool)
   Event OnExitGrindRail() {is=title K2Node_Event}
   > Set KneeStraight(KneeStraight=false)
  ]}
]
variables [
 Backward {type=number}
 Gravity Override Params {type=GravityOverrideParamsStruct}
 GravityOverride {type=Vector}
 grinding {type=boolean}
 inFrontEndNotEmoting {type=boolean}
 Is Backpedaling {type=number}
 Is Crouching {type=number}
 Is Falling {type=number}
 Is Jumping {type=number}
 Is Moving 2D {type=number}
 Is NotIn Front End {type=number}
 Is Sky Diving {type=number}
 Is Skydive Floating {type=number}
 Is Sliding {type=number}
 Is Sprinting {type=number}
 isCrouchingNotMoving {type=number}
 isCrouchingRightMovementClean {type=number}
 isCrouchingRLeftMovementClean {type=number}
 isCrouchMoving {type=number}
 isFallingNoSkydive {type=number}
 isInFrontEndIsEmoting {type=boolean}
 isJumpingNoSkydive {type=number}
 isNotBackpeddling {type=number}
 isNotCrouching {type=number}
 isNotMoving {type=number}
 isNotMovingNotCrouching {type=number}
 isNotSkydiving {type=number}
 isNotSkydivingNotInFE {type=number}
 isNotSkydivingNotSliding {type=number}
 isNotSliding {type=number}
 isNotSprinting {type=number}
 KneeLf {type=number}
 KneeStraight {type=boolean}
 LeftMovement {type=number}
 LeftMovementClean {type=number}
 RightMovement {type=number}
 RightMovementClean {type=number}
 thigh_up_l {type=number}
 thigh_up_r {type=number}
 ThighSideLf {type=number}
 ThighSideRt {type=number}
]
